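# Control Rig for the F_MED_AlderOptic female body cosmetic, adding corrective deformation on top of the FortDeformRig unit.
# A shared GetAlphaFromQuatDelta function remaps swing or twist between two items into a weight that drives ModifyTransforms on forearm twist, ankle, collar and lat helper bones.
# Chest collider bones, grouped as hairColliders, are parent-constrained to torso and arm bones, then aimed at targets added in construction.
# Jacket base bones use weighted parent constraints plus spine-bend correctives.
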ControlRigBlueprint /BRCosmetics/Characters/Player/Female/Medium/Bodies/F_MED_AlderOptic/Meshes/F_MED_AlderOptic_CtrlRig
graphs [
 CollapseNode_ContainedGraph {nodes [
   n1 = RigVMFunctionEntryNode()
   RigVMFunction_Sequence = &RigVMFunction_Sequence(ExecuteContext=@n1.ExecuteContext)
   GetAlphaFromQuatDelta_3 = RigVMFunctionReferenceNode(
    Rotation Order=ZYX
    child=(Type=Bone,Name="clavicle_l")
    debugRotation=false
    outputMinMax=(X=0,Y=1)
    parent=(Type=Null,Name="clavicle_l_null")
    premult=1
    rotationMinMax=(X=0,Y=-40)
    triggerAxis=Y
    twistAxis=(X=1,Y=0,Z=0)
    ExecuteContext=@RigVMFunction_Sequence.A
   )
   > &RigUnit_ModifyTransforms(
    ItemToModify="((Item=(Type=Bone,Name=\"dyn_lathelper_l\"),Transform=(Rotation=(X=0,Y=0,Z=0,W=-1),Translation=(X=-5,Y=0,Z=0),Scale3D=(X=1,Y=1,Z=1))),(Item=(Type=Bone,Name=\"dyn_collarhelper_l\"),Transform=(Rotation=(X=0,Y=-0.2056,Z=0,W=-0.9786),Translation=(X=7.4621,Y=3.9015 ...(truncated)"
    WeightMinimum=0
    Weight=@GetAlphaFromQuatDelta_3.alpha
   )
   > GetAlphaFromQuatDelta_1 = RigVMFunctionReferenceNode(
    Rotation Order=ZYX
    child=(Type=Bone,Name="clavicle_r")
    debugRotation=false
    outputMinMax=(X=0,Y=1)
    parent=(Type=Null,Name="clavicle_r_null")
    premult=1
    rotationMinMax=(X=0,Y=-40)
    triggerAxis=Y
    twistAxis=(X=1,Y=0,Z=0)
   )
   > &RigUnit_ModifyTransforms(
    ItemToModify="((Item=(Type=Bone,Name=\"dyn_lathelper_r\"),Transform=(Rotation=(X=0,Y=0,Z=0,W=-1),Translation=(X=5,Y=0,Z=0),Scale3D=(X=1,Y=1,Z=1))),(Item=(Type=Bone,Name=\"dyn_collarhelper_r\"),Transform=(Rotation=(X=0,Y=-0.2507,Z=0,W=0.9681),Translation=(X=-10.0464,Y=0.3137 ...(truncated)"
    WeightMinimum=0
    Weight=@GetAlphaFromQuatDelta_1.alpha
   )
   > GetAlphaFromQuatDelta = RigVMFunctionReferenceNode(
    Rotation Order=ZYX
    child=(Type=Bone,Name="head")
    outputMinMax=(X=0,Y=1)
    parent=(Type=Null,Name="head_null")
    premult=1
    rotationMinMax=(X=-15,Y=-45)
    triggerAxis=Z
    twistAxis=(X=1,Y=0,Z=0)
   )
   > &RigUnit_ModifyTransforms(
    ItemToModify="((Item=(Type=Bone,Name=\"dyn_collarhelper_r\"),Transform=(Rotation=(X=0,Y=0,Z=0,W=1),Translation=(X=0,Y=0,Z=-2),Scale3D=(X=1,Y=1,Z=1))),(Item=(Type=Bone,Name=\"dyn_collarhelper_l\"),Transform=(Rotation=(X=0,Y=0,Z=0,W=1),Translation=(X=0,Y=0,Z=-5),Scale3D=(X=1, ...(truncated)"
    WeightMinimum=0
    Weight=@GetAlphaFromQuatDelta.alpha
   )
   > GetAlphaFromQuatDelta_2 = RigVMFunctionReferenceNode(
    Rotation Order=ZYX
    child=(Type=Bone,Name="clavicle_l")
    debugRotation=false
    onlyTwist=false
    outputMinMax=(X=0,Y=1)
    parent=(Type=Null,Name="clavicle_l_null")
    premult=1
    rotationMinMax=(X=0,Y=10)
    triggerAxis=Z
    twistAxis=(X=1,Y=0,Z=0)
   )
   Power = /Script/RigVM.RigVMFunction_MathDoublePow(B=3, A=@GetAlphaFromQuatDelta_2.alpha) {is=struct RigVMUnitNode, template="Power::Execute(in A,in B,out Result)"}
   > &RigUnit_ModifyTransforms(
    ItemToModify=((Item=(Type=Bone,Name="dyn_collarhelper_l"),Transform=(Rotation=(X=0,Y=0,Z=0,W=1),Translation=(X=0,Y=-1,Z=-5),Scale3D=(X=1,Y=1,Z=1))))
    WeightMinimum=0
    Weight=@Power.Result
   )
   RigVMFunctionReturnNode(ExecuteContext=@RigVMFunction_Sequence.B)
  ]}
 CollapseNode_ContainedGraph {nodes [
   RigVMFunctionEntryNode()
   > RigVMFunction_Sequence = RigVMAggregateNode()
   GetAlphaFromQuatDelta = RigVMFunctionReferenceNode(
    Rotation Order=ZYX
    child=(Type=Bone,Name="foot_l")
    outputMinMax=(X=0,Y=1)
    parent=(Type=Bone,Name="calf_l")
    premult=1
    rotationMinMax=(X=0,Y=90)
    triggerAxis=Z
    twistAxis=(X=1,Y=0,Z=0)
    ExecuteContext=@RigVMFunction_Sequence.A
   )
   > &RigUnit_ModifyTransforms(
    ItemToModify=((Item=(Type=Bone,Name="dyn_anklehelper_l"),Transform=(Rotation=(X=0,Y=0,Z=0,W=1),Translation=(X=0,Y=-9,Z=-8),Scale3D=(X=1,Y=1,Z=1))))
    WeightMinimum=0
    Weight=@GetAlphaFromQuatDelta.alpha
   )
   > GetAlphaFromQuatDelta_1 = RigVMFunctionReferenceNode(
    Rotation Order=ZYX
    child=(Type=Bone,Name="foot_r")
    outputMinMax=(X=0,Y=1)
    parent=(Type=Bone,Name="calf_r")
    premult=1
    rotationMinMax=(X=0,Y=90)
    triggerAxis=Z
    twistAxis=(X=1,Y=0,Z=0)
   )
   > &RigUnit_ModifyTransforms(
    ItemToModify=((Item=(Type=Bone,Name="dyn_anklehelper_r"),Transform=(Rotation=(X=0,Y=0,Z=0,W=1),Translation=(X=0,Y=-9,Z=-8),Scale3D=(X=1,Y=1,Z=1))))
    WeightMinimum=0
    Weight=@GetAlphaFromQuatDelta_1.alpha
   )
   GetAlphaFromQuatDelta_2 = RigVMFunctionReferenceNode(
    Rotation Order=ZYX
    child=(Type=Bone,Name="foot_l")
    outputMinMax=(X=0,Y=1)
    parent=(Type=Bone,Name="calf_l")
    premult=1
    rotationMinMax=(X=0,Y=-45)
    triggerAxis=Z
    twistAxis=(X=1,Y=0,Z=0)
    ExecuteContext=@RigVMFunction_Sequence.B
   )
   > &RigUnit_ModifyTransforms(
    ItemToModify=((Item=(Type=Bone,Name="dyn_anklehelper_l"),Transform=(Rotation=(X=0,Y=0,Z=0,W=1),Translation=(X=0,Y=6,Z=0),Scale3D=(X=1,Y=1,Z=1))))
    WeightMinimum=0
    Weight=@GetAlphaFromQuatDelta_2.alpha
   )
   > GetAlphaFromQuatDelta_3 = RigVMFunctionReferenceNode(
    Rotation Order=ZYX
    child=(Type=Bone,Name="foot_r")
    outputMinMax=(X=0,Y=1)
    parent=(Type=Bone,Name="calf_r")
    premult=1
    rotationMinMax=(X=0,Y=-45)
    triggerAxis=Z
    twistAxis=(X=1,Y=0,Z=0)
   )
   > &RigUnit_ModifyTransforms(
    ItemToModify=((Item=(Type=Bone,Name="dyn_anklehelper_r"),Transform=(Rotation=(X=0,Y=0,Z=0,W=1),Translation=(X=0,Y=6,Z=0),Scale3D=(X=1,Y=1,Z=1))))
    WeightMinimum=0
    Weight=@GetAlphaFromQuatDelta_3.alpha
   )
   RigVMFunctionReturnNode(ExecuteContext=@RigVMFunction_Sequence.C)
  ]}
 CollapseNode_ContainedGraph {nodes [
   n1 = RigVMFunctionEntryNode()
   &RigUnit_ParentConstraint(
    AdvancedSettings=(RotationOrderForFilter=XZY)
    Child=(Type=Bone,Name="coll_chest_01_r")
    Parents=((Item=(Type=Bone,Name="clavicle_r"),Weight=0.0126),(Item=(Type=Bone,Name="deform_pec_r"),Weight=0.137),(Item=(Type=Bone,Name="spine_05"),Weight=0.3893),(Item=(Type=Bone,Name="spine_04"),Weight=0.461))
    Weight=1
    ExecutePin=@n1.ExecutePin
   )
   > &RigUnit_ParentConstraint(
    AdvancedSettings=(RotationOrderForFilter=XZY)
    Child=(Type=Bone,Name="coll_chest_01_l")
    Parents=((Item=(Type=Bone,Name="deform_pec_l"),Weight=0.0777),(Item=(Type=Bone,Name="spine_05"),Weight=0.3346),(Item=(Type=Bone,Name="spine_04"),Weight=0.5877))
    Weight=1
   )
   > &RigUnit_ParentConstraint(
    AdvancedSettings=(RotationOrderForFilter=XZY)
    Child=(Type=Bone,Name="target_chest_01_r")
    Parents="((Item=(Type=Bone,Name=\"deform_upperarm_up_r\"),Weight=0.034),(Item=(Type=Bone,Name=\"deform_pec_r\"),Weight=0.0584),(Item=(Type=Bone,Name=\"spine_04\"),Weight=0.0941),(Item=(Type=Bone,Name=\"deform_upperarm_fr_r\"),Weight=0.0987),(Item=(Type=Bone,Name=\"upperarm_ ...(truncated)"
    Weight=1
   )
   > &RigUnit_ParentConstraint(
    AdvancedSettings=(RotationOrderForFilter=XZY)
    Child=(Type=Bone,Name="target_chest_01_l")
    Parents="((Item=(Type=Bone,Name=\"deform_upperarm_up_l\"),Weight=0.0146),(Item=(Type=Bone,Name=\"deform_upperarm_fr_l\"),Weight=0.0641),(Item=(Type=Bone,Name=\"upperarm_twist_01_l\"),Weight=0.0649),(Item=(Type=Bone,Name=\"deform_pec_l\"),Weight=0.0981),(Item=(Type=Bone,Nam ...(truncated)"
    Weight=1
   )
   > &RigUnit_ParentConstraint(
    AdvancedSettings=(RotationOrderForFilter=XZY)
    Child=(Type=Bone,Name="target_breast_1_r")
    Parents=((Item=(Type=Bone,Name="deform_pec_r"),Weight=0.0801),(Item=(Type=Bone,Name="spine_05"),Weight=0.2325),(Item=(Type=Bone,Name="spine_04"),Weight=0.6874))
    Weight=1
   )
   > &RigUnit_ParentConstraint(
    AdvancedSettings=(RotationOrderForFilter=XZY)
    Child=(Type=Bone,Name="coll_breast_1_r")
    Parents="((Item=(Type=Bone,Name=\"spine_02\"),Weight=0.0189),(Item=(Type=Bone,Name=\"upperarm_twist_01_r\"),Weight=0.044),(Item=(Type=Bone,Name=\"clavicle_r\"),Weight=0.0592),(Item=(Type=Bone,Name=\"spine_05\"),Weight=0.1244),(Item=(Type=Bone,Name=\"deform_pec_r\"),Weight=0. ...(truncated)"
    Weight=1
   )
   > internalName0_1 = &RigUnit_AimConstraintLocalSpaceOffset(
    AdvancedSettings=(DebugSettings=(bEnabled=True,Scale=10,WorldOffset=(Rotation=(X=0,Y=0,Z=0,W=1),Translation=(X=0,Y=0,Z=0),Scale3D=(X=1,Y=1,Z=1))),RotationOrderForFilter=XZY)
    Child=(Type=Bone,Name="coll_breast_1_r")
    Filter=(bX=True,bY=True,bZ=True)
    Parents=((Item=(Type=Bone,Name="target_breast_1_r"),Weight=1))
    Weight=1
    WorldUp=(Target=(X=0.9967,Y=0.0801,Z=0.0142),Kind=Direction,Space=(Type=Bone,Name="spine_05"))
    bMaintainOffset=False
   )
   > internalName0 = &RigUnit_AimConstraintLocalSpaceOffset(
    AdvancedSettings=(DebugSettings=(bEnabled=True,Scale=10,WorldOffset=(Rotation=(X=0,Y=0,Z=0,W=1),Translation=(X=0,Y=0,Z=0),Scale3D=(X=1,Y=1,Z=1))),RotationOrderForFilter=XZY)
    Child=(Type=Bone,Name="coll_chest_01_r")
    Filter=(bX=True,bY=True,bZ=True)
    Parents=((Item=(Type=Bone,Name="target_chest_01_r"),Weight=1))
    Weight=1
    WorldUp=(Target=(X=0.3608,Y=0.9074,Z=0.2155),Kind=Direction,Space=(Type=Bone,Name="spine_05"))
    bMaintainOffset=False
   )
   > internalName0_2 = &RigUnit_AimConstraintLocalSpaceOffset(
    AdvancedSettings=(DebugSettings=(bEnabled=True,Scale=10,WorldOffset=(Rotation=(X=0,Y=0,Z=0,W=1),Translation=(X=0,Y=0,Z=0),Scale3D=(X=1,Y=1,Z=1))),RotationOrderForFilter=XZY)
    Child=(Type=Bone,Name="coll_chest_01_l")
    Filter=(bX=True,bY=True,bZ=True)
    Parents=((Item=(Type=Bone,Name="target_chest_01_l"),Weight=1))
    Weight=1
    WorldUp=(Target=(X=0.6583,Y=0.7516,Z=-0.0413),Kind=Direction,Space=(Type=Bone,Name="spine_05"))
    bMaintainOffset=False
   )
   RigVMFunctionReturnNode(ExecutePin=@internalName0_2.ExecutePin)
  ]}
 CollapseNode_ContainedGraph {nodes [
   n1 = RigVMFunctionEntryNode()
   &RigUnit_ParentConstraint(
    AdvancedSettings=(RotationOrderForFilter=XZY)
    Child=(Type=Bone,Name="base_bckJacket")
    Parents="((Item=(Type=Bone,Name=\"spine_01\"),Weight=0.0253),(Item=(Type=Bone,Name=\"spine_03\"),Weight=0.0787),(Item=(Type=Bone,Name=\"spine_02\"),Weight=0.0794),(Item=(Type=Bone,Name=\"clavicle_r\"),Weight=0.0914),(Item=(Type=Bone,Name=\"clavicle_l\"),Weight=0.1102),(Item= ...(truncated)"
    Weight=1
    ExecutePin=@n1.ExecutePin
   )
   > &RigUnit_ParentConstraint(
    AdvancedSettings=(RotationOrderForFilter=XZY)
    Child=(Type=Bone,Name="base_frontJacket")
    Parents=((Item=(Type=Bone,Name="spine_05"),Weight=0.014),(Item=(Type=Bone,Name="spine_03"),Weight=0.068),(Item=(Type=Bone,Name="spine_02"),Weight=0.0729),(Item=(Type=Bone,Name="spine_04"),Weight=0.8451))
    Weight=1
   )
   > GetAlphaFromQuatDelta = RigVMFunctionReferenceNode(
    Rotation Order=ZYX
    child=(Type=Bone,Name="spine_05")
    outputMinMax=(X=0,Y=1)
    parent=(Type=Bone,Name="spine_04")
    premult=1
    rotationMinMax=(X=-18,Y=-30)
    triggerAxis=Z
    twistAxis=(X=1,Y=0,Z=0)
   )
   > &RigUnit_ModifyTransforms(
    ItemToModify=((Item=(Type=Bone,Name="base_bckJacket"),Transform=(Rotation=(X=0.166,Y=0,Z=0,W=-0.9861),Translation=(X=0,Y=0,Z=0),Scale3D=(X=1,Y=1,Z=1))))
    WeightMinimum=0
    Weight=@GetAlphaFromQuatDelta.alpha
   )
   > GetAlphaFromQuatDelta_1 = RigVMFunctionReferenceNode(
    Rotation Order=ZYX
    child=(Type=Bone,Name="spine_04")
    outputMinMax=(X=0,Y=1)
    parent=(Type=Bone,Name="spine_02")
    premult=1
    rotationMinMax=(X=15,Y=30)
    triggerAxis=Z
    twistAxis=(X=1,Y=0,Z=0)
   )
   > n2 = &RigUnit_ModifyTransforms(
    ItemToModify=((Item=(Type=Bone,Name="base_frontJacket"),Transform=(Rotation=(X=0.2164,Y=0,Z=0,W=0.9763),Translation=(X=0,Y=0,Z=1),Scale3D=(X=1,Y=1,Z=1))))
    WeightMinimum=0
    Weight=@GetAlphaFromQuatDelta_1.alpha
   )
   RigVMFunctionReturnNode(ExecutePin=@n2.ExecutePin)
  ]}
 CollapseNode_ContainedGraph {nodes [
   RigVMFunctionEntryNode()
   > GetAlphaFromQuatDelta = RigVMFunctionReferenceNode(
    Rotation Order=ZYX
    child=(Type=Bone,Name="hand_l")
    onlyTwist=True
    outputMinMax=(X=0,Y=1)
    parent=(Type=Null,Name="hand_l_null")
    premult=1
    rotationMinMax=(X=120,Y=150)
    twistAxis=(X=1,Y=0,Z=0)
   )
   > &RigUnit_ModifyTransforms(
    ItemToModify="((Item=(Type=Bone,Name=\"lowerarm_twist_01_l\"),Transform=(Rotation=(X=0,Y=-0.0262,Z=0,W=0.9997),Translation=(X=0,Y=0,Z=1),Scale3D=(X=1,Y=1.2,Z=1.2))),(Item=(Type=Bone,Name=\"lowerarm_twist_02_l\"),Transform=(Rotation=(X=0.2588,Y=0,Z=0,W=0.9659),Translation=(X ...(truncated)"
    WeightMinimum=0
    Weight=@GetAlphaFromQuatDelta.alpha
   )
   > RigVMFunctionReturnNode()
  ]}
 New Function_ContainedGraph {nodes [
   RigVMFunctionEntryNode()
   RigVMFunction_MathBoolToInteger = &RigVMFunction_MathBoolToInteger(Value=RigVMVariableNode() {variable=onlyTwist}.Value)
   RigVMFunction_MathQuaternionSwingTwist = &RigVMFunction_MathQuaternionSwingTwist(
    Input=&RigUnit_GetRelativeTransformForItem(
     bChildInitial=False
     bParentInitial=False
     child=RigVMVariableNode() {variable=child}.Value
     parent=RigVMVariableNode() {variable=parent}.Value
    ).RelativeTransform.Rotation
    twistAxis=RigVMVariableNode() {variable=twistAxis}.Value
   )
   n1 = RigVMDispatchNode(
    Values=((),())
    Index=@RigVMFunction_MathBoolToInteger.Result
    Values.0=@RigVMFunction_MathQuaternionSwingTwist.Swing
    Values.1=@RigVMFunction_MathQuaternionSwingTwist.Twist
   ) {template="DISPATCH_RigVMDispatch_SelectInt32(in Index,in Values,out Result)"}
   > RigVMVariableNode(Value=@n1.Result) {variable=entry_outputQuat}
   RigVMFunction_MathQuaternionToEuler = &RigVMFunction_MathQuaternionToEuler(Value=@n1.Result, RotationOrder=RigVMVariableNode() {variable=Rotation Order}.Value)
   > DISPATCH_RigVMDispatch_Print = RigVMDispatchNode(
    ScreenColor=(R=1,G=1,B=1,A=1)
    ScreenDuration=1.5
    Value=@RigVMFunction_MathQuaternionToEuler.Result
    Enabled=RigVMVariableNode() {variable=debugRotation}.Value
   ) {template="DISPATCH_RigVMDispatch_Print(in Prefix,in Value,in Enabled,in ScreenDuration,in ScreenColor)"}
   Cast = RigVMDispatchNode(Value=RigVMVariableNode() {variable=triggerAxis}.Value) {template="DISPATCH_RigVMDispatch_CastEnumToInt(in Value,out Result)"}
   DISPATCH_RigVMDispatch_SelectInt32 = RigVMDispatchNode(
    Values=(0,0,0)
    Index=@Cast.Result
    Values.0=@RigVMFunction_MathQuaternionToEuler.Result.X
    Values.1=@RigVMFunction_MathQuaternionToEuler.Result.Y
    Values.2=@RigVMFunction_MathQuaternionToEuler.Result.Z
   ) {template="DISPATCH_RigVMDispatch_SelectInt32(in Index,in Values,out Result)"}
   n2 = RigVMVariableNode() {variable=rotationMinMax}
   n3 = RigVMVariableNode() {variable=outputMinMax}
   Remap = &RigVMFunction_MathDoubleRemap(
    bClamp=False
    Value=@DISPATCH_RigVMDispatch_SelectInt32.Result
    SourceMinimum=@n2.Value.X
    SourceMaximum=@n2.Value.Y
    TargetMinimum=@n3.Value.X
    TargetMaximum=@n3.Value.Y
   )
   Absolute = &RigVMFunction_MathDoubleAbs(Value=@Remap.Result)
   n4 = RigVMVariableNode() {variable=absoluteValue}
   Multiply = &RigVMFunction_MathDoubleMul(
    A=&RigVMFunction_MathDoubleClamp(Value=RigVMDispatchNode(True=@Absolute.Result, False=@Remap.Result, Condition=@n4.Value) {template="DISPATCH_RigVMDispatch_If(in Condition,in True,in False,out Result)"}.Result).Result
    B=RigVMVariableNode() {variable=premult}.Value
   )
   > RigVMFunctionReturnNode(alpha=@Multiply.Result, rawQuat=RigVMVariableNode() {variable=entry_outputQuat}.Value)
  ]}
 RigVMFunctionLibrary {nodes [GetAlphaFromQuatDelta = RigVMCollapseNode(
    ExecuteContext=()
    Rotation Order=ZYX
    child=(Type=Bone,Name="None")
    outputMinMax=(X=0,Y=1)
    parent=(Type=Bone,Name="None")
    premult=1
    rotationMinMax=(X=0,Y=0)
    twistAxis=(X=1,Y=0,Z=0)
   )]}
 RigVMFunction_Sequence_1_ContainedGraph {nodes [
   n1 = RigVMFunctionEntryNode()
   RigVMFunction_Sequence = &RigVMFunction_Sequence(ExecuteContext=@n1.ExecuteContext)
   n2 = &RigVMFunction_Sequence(ExecuteContext=@RigVMFunction_Sequence.B)
   RigVMFunctionReturnNode(A=@RigVMFunction_Sequence.A, B=@n2.A, C=@n2.B)
  ]}
 RigVMModel {nodes [
   RigUnit_BeginExecution = &RigUnit_BeginExecution()
   /Script/FortniteGame.RigUnit_FortDeformRig(
    Bool=(bUpperArm=true,bElbow=true,bWrist=true,bGroin=true,bGlute=true,bKnee=true,bPec=true,bBicep=true,bLat=false)
    Curve="(bLUpperArmUpAlpha=False,bRUpperArmUpAlpha=False,bLUpperArmFrAlpha=False,bRUpperArmFrAlpha=False,bLUpperArmBkAlpha=False,bRUpperArmBkAlpha=False,bLWristUpAlpha=False,bRWristUpAlpha=False,bLWristDnAlpha=False,bRWristDnAlpha=False,bLKneeAlpha=False,bRKneeAlp ...(truncated)"
    Debug=(bUpperArmDebug=false,bWristDebug=false,bLatDebug=false,bGroinGluteDebug=false)
    Value="(UpperArmUp=(Rotation=(X=0.0,Y=-0.0,Z=0.0,W=1.0),Translation=(X=0.0,Y=0.0,Z=8.0),Scale3D=(X=0.70,Y=1.0,Z=1.0)),UpperArmFwdFr=(Rotation=(X=0.0,Y=-0.0,Z=0.0,W=1.0),Translation=(X=0.0,Y=10.0,Z=0.0),Scale3D=(X=0.50,Y=1.0,Z=1.0)),UpperArmFwdBk=(Rotation=(X=0.0, ...(truncated)"
    ExecutePin=@RigUnit_BeginExecution.ExecutePin
   ) {is=struct RigVMUnitNode}
   > armTwistFix = RigVMCollapseNode()
   > Ankle push setup = RigVMCollapseNode()
   > CollarSetup = RigVMCollapseNode()
   hairColliders = RigVMCollapseNode(ExecutePin=@CollarSetup.ExecuteContext)
   Jacket_Constraints = RigVMCollapseNode(ExecutePin=@hairColliders.ExecutePin)
  ]}
 RigVMModel Rig {nodes [
   n1 = &RigUnit_PrepareForExecution()
   internalNametarget_chest_01_r = /Script/ControlRig.RigUnit_HierarchyAddBone(
    Name=target_chest_01_r
    Parent=(Type=Bone,Name="spine_05")
    Space=LocalSpace
    Transform=(Rotation=(X=-0.6168,Y=-0.1089,Z=-0.4096,W=0.6633),Translation=(X=6.3189,Y=4.8516,Z=9.909),Scale3D=(X=1,Y=1,Z=1))
    ExecutePin=@n1.ExecutePin
   ) {is=struct RigVMUnitNode}
   > internalNametarget_chest_01_l = &RigUnit_HierarchyAddBone(
    Name=target_chest_01_l
    Parent=(Type=Bone,Name="spine_05")
    Space=LocalSpace
    Transform=(Rotation=(X=-0.5599,Y=0.4551,Z=-0.0255,W=0.6919),Translation=(X=5.1788,Y=5.3569,Z=-8.8043),Scale3D=(X=1,Y=1,Z=1))
   )
   > internalNametarget_breast_1_r = &RigUnit_HierarchyAddBone(
    Name=target_breast_1_r
    Parent=(Type=Bone,Name="spine_05")
    Space=LocalSpace
    Transform=(Rotation=(X=-0.0004,Y=0.7021,Z=0.0567,W=0.7099),Translation=(X=-4.7776,Y=9.9122,Z=2.2158),Scale3D=(X=1,Y=1,Z=1))
   )
   > internalNamehand_l = &RigUnit_HierarchyAddNull(
    Name=hand_l_null
    Parent=(Type=Bone,Name="lowerarm_l")
    Space=LocalSpace
    Transform=(Rotation=(X=0.6784,Y=0.0223,Z=-0.0735,W=0.7307),Translation=(X=22.7355,Y=-0.3203,Z=-0.6105),Scale3D=(X=1,Y=1,Z=1))
   )
  ]}
]
hierarchy [
 root {loc=<0 0 0>, rot=<0 0 0 1>}
 pelvis {parent=root, loc=<0 -0.5301 98.0571>, rot=<0.0127 0.707 -0.0127 -0.707>}
 spine_01 {parent=pelvis, loc=<6.1208 -0.9328 0>, rot=<0 0 0.034 -0.9994>}
 spine_02 {parent=spine_01, loc=<6.5997 -0.2095 0>, rot=<0 0 0.0238 -0.9997>}
 spine_03 {parent=spine_02, loc=<6.4914 -0.1237 0>, rot=<0 0 0.0149 -0.9999>}
 spine_04 {parent=spine_03, loc=<6.9598 0.0794 0>, rot=<0 0 -0.0132 -0.9999>}
 spine_05 {parent=spine_04, loc=<6.9987 0 0>, rot=<0 0 -0.0015 -1>}
 neck_01 {parent=spine_05, loc=<11.3138 0.1529 0>, rot=<0 0 -0.1994 -0.9799>}
 neck_02 {parent=neck_01, loc=<4.5627 0 0>, rot=<0 0 0.0504 -0.9987>}
 head {parent=neck_02, loc=<5.3914 0 0>, rot=<0 0 0.1132 -0.9936>}
 FACIAL_C_FacialRoot {parent=head, loc=<2.1462 -0.7318 -0.0002>, rot=<0.4875 -0.5122 -0.4875 -0.5122>}
 FACIAL_C_UnderChin {parent=FACIAL_C_FacialRoot, loc=<0.0639 8.9308 4.192>, rot=<0.488 0.0002 0.0011 -0.8728>}
 FACIAL_L_UnderChin {parent=FACIAL_C_FacialRoot, loc=<3.0674 7.5128 2.7146>, rot=<0.3907 -0.3557 0.0785 -0.8454>}
 FACIAL_R_UnderChin {parent=FACIAL_C_FacialRoot, loc=<-2.8062 7.6241 2.6707>, rot=<0.3914 0.3562 -0.0764 -0.845>}
 FACIAL_L_Jawline {parent=FACIAL_C_FacialRoot, loc=<4.2875 5.7955 4.0326>, rot=<0.1964 0.159 -0.0515 -0.9662>}
 FACIAL_R_Jawline {parent=FACIAL_C_FacialRoot, loc=<-4.1452 5.8193 3.9748>, rot=<0.196 -0.158 0.0535 -0.9663>}
 FACIAL_C_Jaw {parent=FACIAL_C_FacialRoot, loc=<-0.0079 2.7306 1.8383>, rot=<0.3292 0.0004 0.0011 -0.9442>}
 FACIAL_C_Jawline {parent=FACIAL_C_Jaw, loc=<0.0397 2.2134 7.315>, rot=<0.2586 0 0 -0.966>}
 head_null {parent=neck_02, type=Null, loc=<5.3914 0 0>, rot=<0 0 0.1132 -0.9936>}
 clavicle_l {parent=spine_05, loc=<5.4831 -0.1053 -3.2946>, rot=<0.0465 -0.7289 -0.009 -0.6829>}
 upperarm_l {parent=clavicle_l, loc=<9.8658 -1.2136 -0.4105>, rot=<0.0164 -0.3245 -0.0269 -0.9454>}
 lowerarm_l {parent=upperarm_l, loc=<26.465 -0.8345 0.259>, rot=<0.0036 0.0161 -0.2298 -0.9731>}
 hand_l {parent=lowerarm_l, loc=<22.7355 -0.3203 -0.6105>, rot=<-0.6784 -0.0223 0.0735 -0.7307>}
 thumb_01_l {parent=hand_l, loc=<1.3118 -0.3191 -2.459>, rot=<0.6546 -0.4042 0.0664 -0.6354>}
 thumb_02_l {parent=thumb_01_l, loc=<5.9336 -0.0448 0.0401>, rot=<-0.0037 0.0213 0.298 -0.9543>}
 thumb_03_l {parent=thumb_02_l, loc=<2.2223 0.3682 -0.0907>, rot=<0.0118 -0.013 0.0066 -0.9998>}
 index_metacarpal_l {parent=hand_l, loc=<1.6535 0.0183 -2.96>, rot=<0.0158 -0.0124 -0.0173 -0.9996>}
 index_01_l {parent=index_metacarpal_l, loc=<8.3008 0.0357 -0.0524>, rot=<-0.0044 0.0485 0.0924 -0.9945>}
 index_02_l {parent=index_01_l, loc=<3.3587 -0.3004 0.0452>, rot=<-0.002 0.0027 0.3225 -0.9466>}
 index_03_l {parent=index_02_l, loc=<2.5578 -0.123 -0.0252>, rot=<-0.002 0.0067 0.2558 -0.9667>}
 middle_metacarpal_l {parent=hand_l, loc=<2.1856 -0.1725 -0.5984>, rot=<0.006 0.0168 -0.0329 -0.9993>}
 middle_01_l {parent=middle_metacarpal_l, loc=<7.3275 -0.2447 0.0129>, rot=<-0.0208 0.0042 0.0858 -0.9961>}
 middle_02_l {parent=middle_01_l, loc=<4.0123 0.1401 0.1175>, rot=<-0.0218 0.0258 0.3512 -0.9357>}
 middle_03_l {parent=middle_02_l, loc=<2.6077 0.0534 0.1373>, rot=<0.0082 0.0188 0.1708 -0.9851>}
 ring_metacarpal_l {parent=hand_l, loc=<2.2074 -0.3625 1.3638>, rot=<-0.0284 0.0722 -0.0458 -0.9959>}
 ring_01_l {parent=ring_metacarpal_l, loc=<7.0295 -0.6197 -0.1446>, rot=<-0.0489 -0.0252 0.1892 -0.9804>}
 ring_02_l {parent=ring_01_l, loc=<3.6174 0.4647 -0.1027>, rot=<0 0 0.2806 -0.9598>}
 ring_03_l {parent=ring_02_l, loc=<2.4152 0.114 0.0199>, rot=<0 0 0.1189 -0.9929>}
 pinky_metacarpal_l {parent=hand_l, loc=<1.7616 -0.9753 3.3814>, rot=<-0.0397 0.0814 -0.069 -0.9935>}
 pinky_01_l {parent=pinky_metacarpal_l, loc=<7.1709 -0.4831 -0.1094>, rot=<-0.0654 -0.0035 0.2242 -0.9723>}
 pinky_02_l {parent=pinky_01_l, loc=<2.8274 0.0893 -0.1043>, rot=<0.0037 0.0011 0.2854 -0.9584>}
 pinky_03_l {parent=pinky_02_l, loc=<2.2046 0.0307 -0.0078>, rot=<-0.0002 0.0098 0.1445 -0.9895>}
 deform_wrist_dn_base_l {parent=hand_l, loc=<0 -1.5326 0>, rot=<0 0 0 1>}
 deform_wrist_dn_l {parent=deform_wrist_dn_base_l, loc=<0 0 0>, rot=<0 0 0 1>}
 deform_wrist_up_base_l {parent=hand_l, loc=<0 2.3009 0>, rot=<0 0 0 1>}
 deform_wrist_up_l {parent=deform_wrist_up_base_l, loc=<0 0 0>, rot=<0 0 0 1>}
 lowerarm_twist_02_l {parent=lowerarm_l, loc=<10.6707 0 0>, rot=<0 0 0 1>}
 lowerarm_twist_01_l {parent=lowerarm_l, loc=<18.316 0 0>, rot=<0 0 0 1>}
 deform_elbow_in_base_l {parent=lowerarm_l, loc=<0.1834 3.6062 -0.2572>, rot=<0 0 0 1>}
 deform_elbow_in_l {parent=deform_elbow_in_base_l, loc=<0 0 0>, rot=<0 0 0 1>}
 deform_elbow_out_base_l {parent=lowerarm_l, loc=<0 -4 0>, rot=<0 0 0 1>}
 deform_elbow_out_l {parent=deform_elbow_out_base_l, loc=<0 0 0>, rot=<0 0 0 1>}
 upperarm_twist_01_l {parent=upperarm_l, loc=<7.2135 -0.1134 0>, rot=<0 0 0 1>}
 deform_upperarm_bk_base_l {parent=upperarm_twist_01_l, loc=<-6.4053 -4.3746 0.314>, rot=<0 0 0 1>}
 deform_upperarm_bk_l {parent=deform_upperarm_bk_base_l, loc=<0 0 0>, rot=<0 0 0 1>}
 deform_upperarm_fr_base_l {parent=upperarm_twist_01_l, loc=<-5.6312 5.092 0.3354>, rot=<0 0 0 1>}
 deform_upperarm_fr_l {parent=deform_upperarm_fr_base_l, loc=<0 0 0>, rot=<0 0 0 1>}
 deform_upperarm_up_base_l {parent=upperarm_twist_01_l, loc=<-8.891 0.4825 3.9517>, rot=<0 0 0 1>}
 deform_upperarm_up_l {parent=deform_upperarm_up_base_l, loc=<0 0 0>, rot=<0 0 0 1>}
 upperarm_twist_02_l {parent=upperarm_l, loc=<19.7998 -0.1134 0>, rot=<0 0 0 1>}
 base_collarhelper_l {parent=clavicle_l, loc=<6.4385 0.3803 4.2962>, rot=<0.0008 0.0336 -0.0252 -0.9991>}
 dyn_collarhelper_l {parent=base_collarhelper_l, loc=<0 0 0>, rot=<0 0 0 1>}
 clavicle_r {parent=spine_05, loc=<5.4831 -0.1053 3.2946>, rot=<0.7289 0.0465 -0.6829 0.009>}
 upperarm_r {parent=clavicle_r, loc=<-9.8658 1.2136 0.4105>, rot=<0.0164 -0.3245 -0.0269 -0.9454>}
 lowerarm_r {parent=upperarm_r, loc=<-26.465 0.8345 -0.259>, rot=<0.0036 0.0161 -0.2298 -0.9731>}
 hand_r {parent=lowerarm_r, loc=<-22.7355 0.3203 0.6106>, rot=<-0.6784 -0.0223 0.0735 -0.7307>}
 thumb_01_r {parent=hand_r, loc=<-1.3118 0.3191 2.459>, rot=<0.6546 -0.4042 0.0664 -0.6354>}
 thumb_02_r {parent=thumb_01_r, loc=<-5.9336 0.0448 -0.0401>, rot=<-0.0037 0.0213 0.298 -0.9543>}
 thumb_03_r {parent=thumb_02_r, loc=<-2.2223 -0.3682 0.0907>, rot=<0.0118 -0.013 0.0066 -0.9998>}
 index_metacarpal_r {parent=hand_r, loc=<-1.6535 -0.0183 2.96>, rot=<0.0158 -0.0124 -0.0173 -0.9996>}
 index_01_r {parent=index_metacarpal_r, loc=<-8.3008 -0.0357 0.0524>, rot=<-0.0044 0.0485 0.0924 -0.9945>}
 index_02_r {parent=index_01_r, loc=<-3.3587 0.3004 -0.0452>, rot=<-0.002 0.0027 0.3225 -0.9466>}
 index_03_r {parent=index_02_r, loc=<-2.5578 0.123 0.0252>, rot=<-0.002 0.0067 0.2558 -0.9667>}
 middle_metacarpal_r {parent=hand_r, loc=<-2.1856 0.1725 0.5984>, rot=<0.006 0.0168 -0.0329 -0.9993>}
 middle_01_r {parent=middle_metacarpal_r, loc=<-7.3275 0.2447 -0.0129>, rot=<-0.0208 0.0042 0.0858 -0.9961>}
 middle_02_r {parent=middle_01_r, loc=<-4.0123 -0.1401 -0.1175>, rot=<-0.0218 0.0258 0.3512 -0.9357>}
 middle_03_r {parent=middle_02_r, loc=<-2.6077 -0.0534 -0.1373>, rot=<0.0082 0.0188 0.1708 -0.9851>}
 ring_metacarpal_r {parent=hand_r, loc=<-2.2074 0.3625 -1.3638>, rot=<-0.0284 0.0722 -0.0458 -0.9959>}
 ring_01_r {parent=ring_metacarpal_r, loc=<-7.0295 0.6197 0.1446>, rot=<-0.0489 -0.0252 0.1892 -0.9804>}
 ring_02_r {parent=ring_01_r, loc=<-3.6174 -0.4647 0.1027>, rot=<0 0 0.2806 -0.9598>}
 ring_03_r {parent=ring_02_r, loc=<-2.4152 -0.114 -0.0199>, rot=<0 0 0.1189 -0.9929>}
 pinky_metacarpal_r {parent=hand_r, loc=<-1.7616 0.9753 -3.3814>, rot=<-0.0397 0.0814 -0.069 -0.9935>}
 pinky_01_r {parent=pinky_metacarpal_r, loc=<-7.1709 0.4831 0.1094>, rot=<-0.0654 -0.0035 0.2242 -0.9723>}
 pinky_02_r {parent=pinky_01_r, loc=<-2.8274 -0.0893 0.1043>, rot=<0.0037 0.0011 0.2854 -0.9584>}
 pinky_03_r {parent=pinky_02_r, loc=<-2.2046 -0.0307 0.0078>, rot=<-0.0002 0.0098 0.1445 -0.9895>}
 deform_wrist_dn_base_r {parent=hand_r, loc=<-0.0001 1.5328 0>, rot=<0 0 0 -1>}
 deform_wrist_dn_r {parent=deform_wrist_dn_base_r, loc=<0 0 0>, rot=<0 0 0 1>}
 deform_wrist_up_base_r {parent=hand_r, loc=<-0.0002 -2.3007 -0.0001>, rot=<0 0 0 -1>}
 deform_wrist_up_r {parent=deform_wrist_up_base_r, loc=<0 0 0>, rot=<0 0 0 1>}
 lowerarm_twist_02_r {parent=lowerarm_r, loc=<-10.6707 0 0>, rot=<0 0 0 -1>}
 lowerarm_twist_01_r {parent=lowerarm_r, loc=<-18.316 0 0>, rot=<0 0 0 -1>}
 deform_elbow_in_base_r {parent=lowerarm_r, loc=<-0.1834 -3.6062 0.2572>, rot=<0 0 0 -1>}
 deform_elbow_in_r {parent=deform_elbow_in_base_r, loc=<0 0 0>, rot=<0 0 0 1>}
 deform_elbow_out_base_r {parent=lowerarm_r, loc=<0 4 -0.0001>, rot=<0 0 0 -1>}
 deform_elbow_out_r {parent=deform_elbow_out_base_r, loc=<0 0 0>, rot=<0 0 0 1>}
 upperarm_twist_01_r {parent=upperarm_r, loc=<-7.2135 0.1134 0>, rot=<0 0 0 -1>}
 deform_upperarm_bk_base_r {parent=upperarm_twist_01_r, loc=<6.4053 4.3746 -0.314>, rot=<0 0 0 -1>}
 deform_upperarm_bk_r {parent=deform_upperarm_bk_base_r, loc=<0 0 0>, rot=<0 0 0 1>}
 deform_upperarm_fr_base_r {parent=upperarm_twist_01_r, loc=<5.6312 -5.092 -0.3354>, rot=<0 0 0 -1>}
 deform_upperarm_fr_r {parent=deform_upperarm_fr_base_r, loc=<0 0 0>, rot=<0 0 0 1>}
 deform_upperarm_up_base_r {parent=upperarm_twist_01_r, loc=<8.891 -0.4825 -3.9517>, rot=<0 0 0 -1>}
 deform_upperarm_up_r {parent=deform_upperarm_up_base_r, loc=<0 0 0>, rot=<0 0 0 1>}
 upperarm_twist_02_r {parent=upperarm_r, loc=<-19.7998 0.1134 0>, rot=<0 0 0 -1>}
 base_collarhelper_r {parent=clavicle_r, loc=<-7.2171 -0.1277 -5.2163>, rot=<0.9991 0.0252 0.0336 0.0008>}
 dyn_collarhelper_r {parent=base_collarhelper_r, loc=<0 0 0>, rot=<0 0 0 1>}
 attach_backpack {parent=spine_05, loc=<-0.1756 -13.4027 0.0245>, rot=<0.0044 -0.7094 -0.0091 -0.7048>}
 Backpack_BR {parent=attach_backpack, type=Null, loc=<0 1.5 0>, rot=<0.0436 0 0 0.999>}
 attach_cape {parent=spine_05, loc=<9.017 -11.714 0>, rot=<0.0616 -0.7044 -0.0616 -0.7044>}
 cape {parent=attach_cape, type=Null, loc=<0 1.7305 0>, rot=<0.0436 0 0 0.999>}
 base_Hood {parent=spine_05, loc=<10.5102 -4.5396 0>, rot=<0.0284 -0.7065 -0.0284 -0.7065>}
 dyn_Hood_01 {parent=base_Hood, loc=<0 0 0>, rot=<-0.1286 -0.1286 0.6953 -0.6953>}
 dyn_Hood_02 {parent=dyn_Hood_01, loc=<4.3248 0 0>, rot=<0 -0.3555 0 -0.9347>}
 clavicle_l_null {parent=spine_05, type=Null, loc=<5.4831 -0.1053 -3.2946>, rot=<0.0465 -0.7289 -0.009 -0.6829>}
 clavicle_r_null {parent=spine_05, type=Null, loc=<5.4831 -0.1053 3.2946>, rot=<0.7289 0.0465 -0.6829 0.009>}
 coll_chest_01_r {parent=spine_05, loc=<-1.2647 9.6927 2.2211>, rot=<0.6168 0.1089 0.4096 -0.6633>}
 coll_chest_01_l {parent=spine_05, loc=<-1.7378 11.8051 -1.6885>, rot=<0.5599 -0.4551 0.0255 -0.6919>}
 coll_breast_1_r {parent=spine_05, loc=<-4.8281 9.3971 8.6449>, rot=<0.0004 -0.7021 -0.0567 -0.7099>}
 deform_pec_aim_base_l {parent=spine_04, loc=<3.2456 9.3309 -1.2237>, rot=<0.0603 -0.4033 0.3318 -0.8507>}
 deform_pec_aim_l {parent=deform_pec_aim_base_l, loc=<0 0 0>, rot=<0 0 0 1>}
 deform_pec_base_l {parent=deform_pec_aim_l, loc=<9.3665 0 0>, rot=<0 0 0 1>}
 deform_pec_l {parent=deform_pec_base_l, loc=<0 0 0>, rot=<0 0 0 1>}
 deform_pec_aim_base_r {parent=spine_04, loc=<3.2452 9.3309 1.2237>, rot=<-0.4033 -0.0603 0.8507 0.3318>}
 deform_pec_aim_r {parent=deform_pec_aim_base_r, loc=<0 0 0>, rot=<0 0 0 1>}
 deform_pec_base_r {parent=deform_pec_aim_r, loc=<-9.3667 -0.0003 -0.0004>, rot=<0 0 0 1>}
 deform_pec_r {parent=deform_pec_base_r, loc=<0 0 0>, rot=<0 0 0 1>}
 base_Zipper_01 {parent=spine_04, loc=<2.4611 13.8359 0.0233>, rot=<0.0294 -0.7065 -0.0294 -0.7065>}
 dyn_Zipper_01 {parent=base_Zipper_01, loc=<0 0 0>, rot=<0.2765 -0.2873 -0.6359 -0.6607>}
 base_lathelper_r {parent=spine_04, loc=<0.4117 3.0325 8.8876>, rot=<0.0294 -0.7065 -0.0294 -0.7065>}
 dyn_lathelper_r {parent=base_lathelper_r, loc=<0 0 0>, rot=<0 0 0 1>}
 base_lathelper_l {parent=spine_04, loc=<0.4117 3.0325 -8.3073>, rot=<0.0294 -0.7065 -0.0294 -0.7065>}
 dyn_lathelper_l {parent=base_lathelper_l, loc=<0 0 0>, rot=<0 0 0 1>}
 base_bckJacket {parent=spine_03, loc=<8.6529 -8.7393 -0.1638>, rot=<0.0387 -0.706 -0.0387 -0.706>}
 dyn_bckJacket {parent=base_bckJacket, loc=<0 0 0>, rot=<0.5552 0.5118 -0.4821 0.4444>}
 base_frontJacket {parent=spine_03, loc=<2.6503 12.8275 -0.0492>, rot=<0.0387 -0.706 -0.0387 -0.706>}
 dyn_frontJacket {parent=base_frontJacket, loc=<0 0 0>, rot=<-0.507 0.4368 0.5629 0.485>}
 thigh_l {parent=pelvis, loc=<-1.0855 -0.2109 -8.7053>, rot=<0.0009 0.0355 0.0155 -0.9993>}
 calf_l {parent=thigh_l, loc=<-42.5284 -0.0019 -0.7297>, rot=<0.0006 -0.0161 -0.0391 -0.9991>}
 foot_l {parent=calf_l, loc=<-46.7401 -2.5014 -0.592>, rot=<0.0427 -0.0238 0.0353 -0.9982>}
 ball_l {parent=foot_l, loc=<-3.5658 10.8066 -0.1544>, rot=<0.0001 -0.0001 -0.7063 -0.7079>}
 base_anklehelper_l {parent=foot_l, loc=<0 0 0>, rot=<-0.0357 -0.7045 -0.0284 -0.7082>}
 dyn_anklehelper_l {parent=base_anklehelper_l, loc=<0 0 0>, rot=<0 0 0 1>}
 calf_twist_02_l {parent=calf_l, loc=<-19.1546 -1.7943 -0.2638>, rot=<0.0237 0.0095 -0.0035 -0.9997>}
 calf_twist_01_l {parent=calf_l, loc=<-35.8483 -1.9197 -0.5794>, rot=<0.0237 0.0095 -0.0035 -0.9997>}
 deform_knee_in_base_l {parent=calf_l, loc=<0.8031 -5.0231 0.2456>, rot=<0 0 0 -1>}
 deform_knee_in_l {parent=deform_knee_in_base_l, loc=<0 0 0>, rot=<0 0 0 1>}
 deform_knee_out_base_l {parent=calf_l, loc=<0 4 0>, rot=<0 0 0 -1>}
 deform_knee_out_l {parent=deform_knee_out_base_l, loc=<0 0 0>, rot=<0 0 0 1>}
 thigh_twist_01_l {parent=thigh_l, loc=<-27.952 0 0>, rot=<0.0013 0 0 -1>}
 deform_groin_base_l {parent=thigh_l, loc=<0.4801 5.6776 -0.2486>, rot=<0 0 0 -1>}
 deform_groin_l {parent=deform_groin_base_l, loc=<0 0 0>, rot=<0 0 0 1>}
 thigh_r {parent=pelvis, loc=<-1.0855 -0.2109 8.7053>, rot=<0.0355 -0.0009 0.9993 0.0155>}
 calf_r {parent=thigh_r, loc=<42.5284 0.0019 0.7297>, rot=<0.0006 -0.0161 -0.0391 -0.9991>}
 foot_r {parent=calf_r, loc=<46.7401 2.5014 0.592>, rot=<0.0427 -0.0238 0.0353 -0.9982>}
 ball_r {parent=foot_r, loc=<3.5658 -10.8066 0.1544>, rot=<0.0001 -0.0001 -0.7063 -0.7079>}
 base_anklehelper_r {parent=foot_r, loc=<0 0 0>, rot=<0.7082 0.0284 -0.7045 -0.0357>}
 dyn_anklehelper_r {parent=base_anklehelper_r, loc=<0 0 0>, rot=<0 0 0 1>}
 calf_twist_02_r {parent=calf_r, loc=<19.1546 1.7943 0.2638>, rot=<0.0237 0.0095 -0.0035 -0.9997>}
 calf_twist_01_r {parent=calf_r, loc=<35.8483 1.9197 0.5794>, rot=<0.0237 0.0095 -0.0035 -0.9997>}
 deform_knee_in_base_r {parent=calf_r, loc=<-0.8031 5.0231 -0.2457>, rot=<0 0 0 -1>}
 deform_knee_in_r {parent=deform_knee_in_base_r, loc=<0 0 0>, rot=<0 0 0 1>}
 deform_knee_out_base_r {parent=calf_r, loc=<0 -4 0>, rot=<0 0 0 -1>}
 deform_knee_out_r {parent=deform_knee_out_base_r, loc=<0 0 0>, rot=<0 0 0 1>}
 thigh_twist_01_r {parent=thigh_r, loc=<27.952 0 0>, rot=<0.0013 0 0 -1>}
 deform_groin_base_r {parent=thigh_r, loc=<-0.4801 -5.6776 0.2486>, rot=<0 0 0 -1>}
 deform_groin_r {parent=deform_groin_base_r, loc=<0 0 0>, rot=<0 0 0 1>}
 deform_glute_base_l {parent=pelvis, loc=<-7.3466 -9.0213 -9.1245>, rot=<0.0172 0.031 -0.4477 -0.8935>}
 deform_glute_l {parent=deform_glute_base_l, loc=<0 0 0>, rot=<0 0 0 1>}
 deform_glute_base_r {parent=pelvis, loc=<-7.3464 -9.0213 9.1245>, rot=<0.031 -0.0172 0.8935 -0.4477>}
 deform_glute_r {parent=deform_glute_base_r, loc=<0.0131 0.0327 0>, rot=<0 0 0 1>}
]
