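# Physics asset for the F_MED_AlderOptic female medium character body, covering the back of the jacket for secondary motion.
# It holds three capsule bodies: a tiny anchor on base_bckJacket, a small capsule on dyn_bckJacket and a large torso capsule on spine_03.
# Constraints chain them, with dyn_bckJacket held to narrow swing and twist limits against its base, while the base rotates freely relative to spine_03.
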PhysicsAsset /BRCosmetics/Characters/Player/Female/Medium/Bodies/F_MED_AlderOptic/Meshes/F_MED_AlderOptic_Physics
bodies {
 base_bckJacket [{shapeType=Capsule, length=0.505, radius=0.505, shapeName=base_bckJacket_capsule, rotation=<rot 0 -90 0>}]
 dyn_bckJacket [{shapeType=Capsule, length=5, radius=2, center=<3 0 1.8652>, rotation=<rot 90 0 0>}]
 spine_03 [{shapeType=Capsule, length=6.6381, radius=12, center=<2.4856 3.8353 -0.1255>, rotation=<rot 90 0 0>}]
}
constraints [
 {bone1Name=dyn_bckJacket, bone2Name=base_bckJacket, swing1LimitDegrees=12, swing2LimitDegrees=15, twistLimitDegrees=5}
 {
  bone1Name=base_bckJacket
  bone2Name=spine_03
  swing1LimitDegrees=45
  swing1Motion=Free
  swing2LimitDegrees=45
  swing2Motion=Free
  twistLimitDegrees=45
  twistMotion=Free
 }
]
