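# Animation Blueprint for the AlderOptic female body cosmetic, adding secondary motion to jacket, zipper and hood.
# A parent-copied pose runs through Control Rig, then a fixed bool selects the Anim Dynamics-only jacket/zipper chain over a RigidBody variant, before hood dynamics.
# Jaw and under-chin bones then take transforms read each update from the Head mesh.
# The event graph also drives velocity-based WPO cloth materials, adding wind while skydive-floating or parachuting.
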AnimBlueprint /BRCosmetics/Characters/Player/Female/Medium/Bodies/F_MED_AlderOptic/Meshes/F_MED_AlderOptic_AnimBP
target_skeleton=&/F_MED_AlderOptic_Skeleton
graphs [
 AnimGraph {nodes [
   "Transform (Modify) Bone - Bone: base_Hood"(Alpha=1, Translation="0,0,0", Rotation="0,0,0", Scale="1,1,1") {boneToModify {boneName=base_Hood}}
   n1 = "Transform (Modify) Bone - Bone: FACIAL_C_UnderChin"!(
    ComponentPose=Local To Component(LocalPose=Use cached pose 'AnimdynCH'())
    bAlphaBoolEnabled:bool=True
    Translation=Get FACIAL_C_UnderChin_Translation() -> (FACIAL_C_UnderChin_Translation:Vector)
    Rotation=Get FACIAL_C_UnderChin_Rotation() -> (FACIAL_C_UnderChin_Rotation:Rotator)
    Scale="1,1,1"
   ) -> (Pose) {
    alphaInputType=Bool
    rotationMode=BMM_Replace
    rotationSpace=BCS_ParentBoneSpace
    translationMode=BMM_Replace
    translationSpace=BCS_ParentBoneSpace
    alphaBoolBlend {blendInTime=0.25, blendOption=Linear, blendOutTime=0.25}
    boneToModify {boneName=FACIAL_C_UnderChin}
   }
   n2 = "Transform (Modify) Bone - Bone: FACIAL_L_UnderChin"!(
    ComponentPose=@n1
    bAlphaBoolEnabled:bool=True
    Translation=Get FACIAL_L_UnderChin_Translation() -> (FACIAL_L_UnderChin_Translation:Vector)
    Rotation=Get FACIAL_L_UnderChin_Rotation() -> (FACIAL_L_UnderChin_Rotation:Rotator)
    Scale="1,1,1"
   ) -> (Pose) {
    alphaInputType=Bool
    rotationMode=BMM_Replace
    rotationSpace=BCS_ParentBoneSpace
    translationMode=BMM_Replace
    translationSpace=BCS_ParentBoneSpace
    alphaBoolBlend {blendOption=Linear}
    boneToModify {boneName=FACIAL_L_UnderChin}
   }
   n3 = "Transform (Modify) Bone - Bone: FACIAL_R_UnderChin"!(
    ComponentPose=@n2
    bAlphaBoolEnabled:bool=True
    Translation=Get FACIAL_R_UnderChin_Translation() -> (FACIAL_R_UnderChin_Translation:Vector)
    Rotation=Get FACIAL_R_UnderChin_Rotation() -> (FACIAL_R_UnderChin_Rotation:Rotator)
    Scale="1,1,1"
   ) -> (Pose) {
    alphaInputType=Bool
    rotationMode=BMM_Replace
    rotationSpace=BCS_ParentBoneSpace
    translationMode=BMM_Replace
    translationSpace=BCS_ParentBoneSpace
    alphaBoolBlend {blendOption=Linear}
    boneToModify {boneName=FACIAL_R_UnderChin}
   }
   n4 = "Transform (Modify) Bone - Bone: FACIAL_L_Jawline"!(
    ComponentPose=@n3
    bAlphaBoolEnabled:bool=True
    Translation=Get FACIAL_L_Jawline_Translation() -> (FACIAL_L_Jawline_Translation:Vector)
    Rotation=Get FACIAL_L_Jawline_Rotation() -> (FACIAL_L_Jawline_Rotation:Rotator)
    Scale="1,1,1"
   ) -> (Pose) {
    alphaInputType=Bool
    rotationMode=BMM_Replace
    rotationSpace=BCS_ParentBoneSpace
    translationMode=BMM_Replace
    translationSpace=BCS_ParentBoneSpace
    alphaBoolBlend {blendOption=Linear}
    boneToModify {boneName=FACIAL_L_Jawline}
   }
   n5 = "Transform (Modify) Bone - Bone: FACIAL_R_Jawline"!(
    ComponentPose=@n4
    bAlphaBoolEnabled:bool=True
    Translation=Get FACIAL_R_Jawline_Translation() -> (FACIAL_R_Jawline_Translation:Vector)
    Rotation=Get FACIAL_R_Jawline_Rotation() -> (FACIAL_R_Jawline_Rotation:Rotator)
    Scale="1,1,1"
   ) -> (Pose) {
    alphaInputType=Bool
    rotationMode=BMM_Replace
    rotationSpace=BCS_ParentBoneSpace
    translationMode=BMM_Replace
    translationSpace=BCS_ParentBoneSpace
    alphaBoolBlend {blendOption=Linear}
    boneToModify {boneName=FACIAL_R_Jawline}
   }
   n6 = "Transform (Modify) Bone - Bone: FACIAL_C_Jaw"!(
    ComponentPose=@n5
    bAlphaBoolEnabled:bool=True
    Translation=Get FACIAL_C_Jaw_Tranform() -> (FACIAL_C_Jaw_Tranform:Vector)
    Rotation=Get FACIAL_C_Jaw_Rotation() -> (FACIAL_C_Jaw_Rotation:Rotator)
    Scale="1,1,1"
   ) -> (Pose) {
    alphaInputType=Bool
    rotationMode=BMM_Replace
    rotationSpace=BCS_ParentBoneSpace
    translationMode=BMM_Replace
    translationSpace=BCS_ParentBoneSpace
    alphaBoolBlend {blendOption=Linear}
    boneToModify {boneName=FACIAL_C_Jaw}
   }
   n7 = "Transform (Modify) Bone - Bone: FACIAL_C_Jawline"!(
    ComponentPose=@n6
    bAlphaBoolEnabled:bool=True
    Translation=Get FACIAL_C_Jawline_Translation() -> (FACIAL_C_Jawline_Translation:Vector)
    Rotation=Get FACIAL_C_Jawline_Rotation() -> (FACIAL_C_Jawline_Rotation:Rotator)
    Scale="1,1,1"
   ) -> (Pose) {
    alphaInputType=Bool
    rotationMode=BMM_Replace
    rotationSpace=BCS_ParentBoneSpace
    translationMode=BMM_Replace
    translationSpace=BCS_ParentBoneSpace
    alphaBoolBlend {blendOption=Linear}
    boneToModify {boneName=FACIAL_C_Jawline}
   }
   n8 = "Transform (Modify) Bone - Bone: dyn_Zipper_01"!(ComponentPose=@n7, bAlphaBoolEnabled:bool=True, Rotation="-15,0,0") -> (Pose) {alphaInputType=Bool, rotationMode=BMM_Additive, rotationSpace=BCS_BoneSpace, boneToModify {boneName=dyn_Zipper_01}, rotation {pitch=-15}}
   Output Pose(Result=Component To Local(ComponentPose=@n8)) {layerGroup=None}
   n9 = Control Rig(Source=Copy Pose From Mesh() {bCopyCurves=true, bUseAttachedParent=true}) {controlRigAssetReference {blueprintRigClass=&/F_MED_AlderOptic_CtrlRig.F_MED_AlderOptic_CtrlRig_C}}
   Save cached pose 'InputCH'(Pose=@n9) {cacheName=InputCH}
   n10 = Anim Dynamics - Bone: dyn_Zipper_01!(ComponentPose=Local To Component(LocalPose=Use cached pose 'InputCH'()), Alpha=1, GravityOverride:Vector="-980,0,0") -> (Pose) {
    angularDampingOverride=1
    angularSpringConstant=100
    bAngularSpring=true
    bOverrideAngularDamping=true
    bUseGravityOverride=true
    gravityScale=0.35
    simulationSpace=RootRelative
    boundBone {boneName=dyn_Zipper_01}
    componentAppliedLinearAccClamp=<6000 6000 6000>
    componentLinearAccScale=<0.5 0.5 0.5>
    componentLinearVelScale=<0.6 0.6 0.6>
    gravityOverride {x=-980}
    physicsBodyDefinitions [{
      boundBone {boneName=dyn_Zipper_01}
      boxExtents=<10 16 8>
      constraintSetup {angularLimitsMax=<0 5 2>, angularLimitsMin=<0 -70 -2>, angularTarget=<1 0 0>}
      localJointOffset=<4 0 0>
     }]
    relativeSpaceBone {boneName=neck_01}
    simSpaceSettings {maxAngularAcceleration=15, maxAngularVelocity=7, simSpaceAngularAlpha=0.75}
   }
   n11 = Anim Dynamics - Bone: dyn_frontJacket!(ComponentPose=@n10, Alpha=1, GravityOverride:Vector="0,0,-980") -> (Pose) {
    angularDampingOverride=1
    angularSpringConstant=9000
    bAngularSpring=true
    bOverrideAngularDamping=true
    bUseGravityOverride=true
    gravityScale=0.35
    simulationSpace=BoneRelative
    boundBone {boneName=dyn_frontJacket}
    componentAppliedLinearAccClamp=<3000 3000 3000>
    componentLinearAccScale=<1 1 1>
    componentLinearVelScale=<1 1 1>
    gravityOverride {z=-980}
    physicsBodyDefinitions [{
      boundBone {boneName=dyn_frontJacket}
      boxExtents=<20 30 20>
      constraintSetup {angularLimitsMax=<0 5 8>, angularLimitsMin=<0 -25 -8>, angularTarget=<1 0 0>}
      localJointOffset=<2 0 0>
     }]
    relativeSpaceBone {boneName=root}
    simSpaceSettings {maxAngularAcceleration=50, maxAngularVelocity=25}
   }
   n12 = RigidBody!(ComponentPose=@n11, Alpha, OverrideWorldGravity:Vector="0,0,0", ExternalForce) -> (Pose) {
    bOverrideWorldGravity=true
    overridePhysicsAsset=&/F_MED_AlderOptic_Physics
    simulationSpace=BaseBoneSpace
    alphaBoolBlend {blendOption=Linear}
    baseBoneRef {boneName=spine_05}
    componentAppliedLinearAccClamp=<3000 3000 3000>
    componentLinearAccScale=<0.25 0.25 0.25>
    componentLinearVelScale=<0.4 0.4 0.4>
    simSpaceSettings {
     maxAngularAcceleration=1000
     maxAngularVelocity=3
     maxLinearAcceleration=250
     maxLinearVelocity=500
     velocityScaleZ=0.45
     worldAlpha=0.6
     externalLinearDragV {x=1}
    }
   }
   n13 = Component To Local(ComponentPose=@n12)
   n14 = Anim Dynamics - Bone: dyn_Zipper_01(ComponentPose=Local To Component(LocalPose=Use cached pose 'InputCH'()), Alpha=0.6, GravityOverride="0,0,-980") {
    angularDampingOverride=1
    angularSpringConstant=9000
    bAngularSpring=true
    bOverrideAngularDamping=true
    bUseGravityOverride=true
    gravityScale=0.35
    simulationSpace=RootRelative
    boundBone {boneName=dyn_Zipper_01}
    componentAppliedLinearAccClamp=<3000 3000 3000>
    componentLinearAccScale=<0.4 0.4 0.4>
    componentLinearVelScale=<0.4 0.4 0.4>
    gravityOverride {z=-980}
    physicsBodyDefinitions [{
      boundBone {boneName=dyn_Zipper_01}
      boxExtents=<8 20 8>
      constraintSetup {angularLimitsMax=<0 5 2>, angularLimitsMin=<0 -70 -2>, angularTarget=<1 0 0>}
      localJointOffset=<6 0 0>
     }]
    simSpaceSettings {maxAngularAcceleration=20, maxAngularVelocity=10, simSpaceAngularAlpha=1}
   }
   n15 = Anim Dynamics - Bone: dyn_frontJacket(ComponentPose=@n14, Alpha=0.8, GravityOverride="0,0,-980") {
    angularDampingOverride=1
    angularSpringConstant=9000
    bAngularSpring=true
    bOverrideAngularDamping=true
    bUseGravityOverride=true
    gravityScale=0.35
    simulationSpace=RootRelative
    boundBone {boneName=dyn_frontJacket}
    componentAppliedLinearAccClamp=<3000 3000 3000>
    componentLinearAccScale=<0.5 0.5 0.5>
    componentLinearVelScale=<0.5 0.5 0.5>
    gravityOverride {z=-980}
    physicsBodyDefinitions [{
      boundBone {boneName=dyn_frontJacket}
      boxExtents=<20 30 20>
      constraintSetup {angularLimitsMax=<0 5 4>, angularLimitsMin=<0 -25 -4>, angularTarget=<1 0 0>}
      localJointOffset=<1 0 0>
     }]
   }
   n16 = Anim Dynamics - Bone: dyn_bckJacket!(ComponentPose=@n15, Alpha=0.8, GravityOverride:Vector="0,0,-980") -> (Pose) {
    angularDampingOverride=1
    angularSpringConstant=9000
    bAngularSpring=true
    bOverrideAngularDamping=true
    bUseGravityOverride=true
    gravityScale=0.35
    simulationSpace=RootRelative
    boundBone {boneName=dyn_bckJacket}
    componentAppliedLinearAccClamp=<3000 3000 3000>
    componentLinearAccScale=<0.5 0.5 0.5>
    componentLinearVelScale=<0.5 0.5 0.5>
    gravityOverride {z=-980}
    physicsBodyDefinitions [{
      boundBone {boneName=dyn_bckJacket}
      boxExtents=<20 30 20>
      constraintSetup {angularLimitsMax=<0 0.1 8>, angularLimitsMin=<0 -25 -8>, angularTarget=<1 0 0>}
      localJointOffset=<1 0 0>
     }]
   }
   n17 = Component To Local(ComponentPose=@n16)
   n18 = Anim Dynamics - Bone: dyn_Hood_01!(
    ComponentPose=Local To Component(LocalPose=Blend Poses by bool(BlendPose_0=@n17, BlendPose_1=@n13, bActiveValue=true))
    Alpha=1
    GravityOverride:Vector="0,0,-980"
   ) -> (Pose) {
    angularDampingOverride=1
    angularSpringConstant=8000
    bAngularSpring=true
    bOverrideAngularDamping=true
    bUseGravityOverride=true
    gravityScale=0.35
    simulationSpace=RootRelative
    boundBone {boneName=dyn_Hood_01}
    componentAppliedLinearAccClamp=<3000 3000 3000>
    componentLinearAccScale=<1 1 1>
    componentLinearVelScale=<1 1 1>
    gravityOverride {z=-980}
    physicsBodyDefinitions [{
      boundBone {boneName=dyn_Hood_01}
      boxExtents=<20 30 30>
      constraintSetup {angularLimitsMax=<0 3 5>, angularLimitsMin=<0 -10 -5>, angularTarget=<1 0 0>}
      localJointOffset=<4 0 0>
     }]
    relativeSpaceBone {boneName=spine_03}
   }
   Save cached pose 'AnimdynCH'(Pose=Component To Local(ComponentPose=@n18)) {cacheName=AnimdynCH}
  ]}
 EventGraph {nodes [
   Event BlueprintInitializeAnimation()
   n1 = TryGetPawnOwner()
   > Is Valid(InputObject=@n1)
   > n2 = Cast To FortPlayerPawnAthena(Object=@n1)
   > Set PlayerPawn(PlayerPawn:Fort Player Pawn Athena Object Reference=@n2) -> (Output_Get:Fort Player Pawn Athena Object Reference)
   n3 = Event BlueprintUpdateAnimation()
   > K2Node_Composite_1 = VelocityDrivenWPO_2!(InputPin:Exec, InputPin2:float=@n3.DeltaTimeX) -> (then_1:Exec) {is=title K2Node_Composite}
   > K2Node_Composite_3 = Head_Set!(exec:Exec) -> (then:Exec) {is=title K2Node_Composite}
   > n4 = "Gravity Override State Function (Parameter Asset) (Nativized)"(
    CharacterPartAnimInstance=Self-Reference()
    GameTimeInSeconds=GetGameTimeInSeconds()
    Parameters=&/F_MED_AlderOptic_FaceAcc_Parameters
   )
   > Set GravityOverride(GravityOverride=@n4)
   n5 = Event OnCharacterCustomizationLoaded()
   > Set HeadMesh(HeadMesh:Skeletal Mesh Component Budgeted Object Reference=Get Array Item with EFort Custom Part Type(Array=@n5.CustomMeshParts, PartType=Head)) -> (Output_Get:Skeletal Mesh Component Budgeted Object Reference)
   > n6 = Set SkelMesh(SkelMesh=GetPartSkeletalMeshComponent())
   > n7 = Get Dynamic Material Instance(SkeletalMeshComponent=@n6)
   > Set Mat Instance Dynamic(Mat Instance Dynamic=@n7)
   > n8 = GetOrCreateDynamicOverlayMaterial(MeshComponent=@n6, MaterialName=None)
   > Set Overlay Mat Instance Dynamic(Overlay Mat Instance Dynamic=@n8)
  ]}
 VelocityDrivenWPO_2 {
  path=EventGraph.K2Node_Composite_1.VelocityDrivenWPO_2
  nodes [
   n1 = Inputs!() -> (InputPin:Exec, InputPin2:float=~)
   > n2 = Sequence()
   > Is Valid(InputObject=Get PlayerPawn() -> (PlayerPawn:Fort Player Pawn Athena Object Reference))
   > n3 = Branch(Condition=Get bIsInFrontEnd())
   Set WPO Globals Function(
    execute=@n3.else
    Material=Get Mat Instance Dynamic()
    Overlay Material=Get Overlay Mat Instance Dynamic()
    Skeletal Mesh=Get SkelMesh()
    "Local Noise Transform  Location"=Get Local Noise Transform Location() -> (Local Noise Transform Location:Vector)
    Local Noise Offset=Get Local Noise Offset()
   )
   > Set WPO Regional Function!(
    execute
    Material:Material Instance Dynamic Object Reference=Get Mat Instance Dynamic()
    Overlay Material:Material Instance Dynamic Object Reference=Get Overlay Mat Instance Dynamic()
    Skeletal Mesh:Skeletal Mesh Component Object Reference=Get SkelMesh()
    Velocity Transform:Name=spine_05
    Velocity Interp Speed:double=Get Velocity Interp Speed()
    Collision Transform:Name=root
    Collision Transform Axis:EAxis Enum
    Invert Collision Transform Axis:bool=false
    Delta Time X:double=@n1
    Previous Velocity Transform Location:"Vector (by ref)"=Get R1 Velocity Transform Location()
    Previous Velocity Transform Velocity:"Vector (by ref)"=Get R1 Velocity Transform Velocity()
    Region:int=1
   ) -> (then)
   > n4 = Branch(Condition=OR Boolean(A=Get bIsSkydiveFloating(), B=Get bIsParachuteOpened()))
   > Set WPO Pho Wind Intensity(WPO Pho Wind Intensity=0.8)
   > n5 = Set WPO Pho Turbulence(WPO Pho Turbulence:double=4.2) -> (Output_Get:double)
   SetScalarParameterValue(
    execute=[@n6.then, @n5.then]
    self=[Get Overlay Mat Instance Dynamic(), Get Mat Instance Dynamic()]
    ParameterName=ClothSim_WindStrength
    Value=Get WPO Pho Wind Intensity()
   )
   > SetScalarParameterValue(
    self=[Get Overlay Mat Instance Dynamic(), Get Mat Instance Dynamic()]
    ParameterName=ClothSim_R1_TurbulenceStrength
    Value=Get WPO Pho Turbulence() -> (WPO Pho Turbulence:double)
   )
   Set WPO Pho Wind Intensity(execute=@n4.else, WPO Pho Wind Intensity=0)
   > n6 = Set WPO Pho Turbulence(WPO Pho Turbulence=0)
   Outputs!(then_1:Exec=@n2.then_1)
  ]
 }
 Head_Set {
  path=EventGraph.K2Node_Composite_3.Head_Set
  nodes [
   Inputs!() -> (exec:Exec)
   > n1 = Sequence()
   n2 = Get HeadMesh() -> (HeadMesh:Skeletal Mesh Component Budgeted Object Reference)
   > Is Valid(InputObject=@n2)
   n3 = GetSocketBoneName!(self:Skinned Mesh Component Object Reference=@n2, InSocketName:Name=FACIAL_C_Jaw) -> (ReturnValue:Name=None)
   > n4 = Set isUsingFacialBones(isUsingFacialBones:bool="Not Equal Exactly (String)"(A="To String (Name)"(InName=@n3), B=None)) -> (Output_Get:bool)
   > Branch(Condition=@n4)
   n5 = Get HeadMesh()
   n6 = BreakTransform(InTransform=GetSocketTransform(self=@n5, InSocketName=FACIAL_C_UnderChin, TransformSpace=RTS_ParentBoneSpace))
   > Set FACIAL_C_UnderChin_Translation(FACIAL_C_UnderChin_Translation:Vector=@n6) -> (Output_Get:Vector)
   > Set FACIAL_C_UnderChin_Rotation(FACIAL_C_UnderChin_Rotation:Rotator=@n6.Rotation) -> (Output_Get:Rotator)
   n7 = BreakTransform(InTransform=GetSocketTransform(self=@n5, InSocketName=FACIAL_C_Chin, TransformSpace=RTS_ParentBoneSpace))
   > Set FACIAL_C_Chin_Translation(FACIAL_C_Chin_Translation:Vector=@n7) -> (Output_Get:Vector)
   > Set FACIAL_C_Chin_Rotation(FACIAL_C_Chin_Rotation:Rotator=@n7.Rotation) -> (Output_Get:Rotator)
   n8 = BreakTransform(InTransform=GetSocketTransform(self=@n5, InSocketName=FACIAL_L_UnderChin, TransformSpace=RTS_ParentBoneSpace))
   > Set FACIAL_L_UnderChin_Translation(FACIAL_L_UnderChin_Translation:Vector=@n8) -> (Output_Get:Vector)
   > Set FACIAL_L_UnderChin_Rotation(FACIAL_L_UnderChin_Rotation:Rotator=@n8.Rotation) -> (Output_Get:Rotator)
   n9 = BreakTransform(InTransform=GetSocketTransform(self=@n5, InSocketName=FACIAL_R_UnderChin, TransformSpace=RTS_ParentBoneSpace))
   > Set FACIAL_R_UnderChin_Translation(FACIAL_R_UnderChin_Translation:Vector=@n9) -> (Output_Get:Vector)
   > Set FACIAL_R_UnderChin_Rotation(FACIAL_R_UnderChin_Rotation:Rotator=@n9.Rotation) -> (Output_Get:Rotator)
   n10 = BreakTransform(InTransform=GetSocketTransform(self=@n5, InSocketName=FACIAL_C_Jaw, TransformSpace=RTS_ParentBoneSpace))
   > Set FACIAL_C_Jaw_Tranform(FACIAL_C_Jaw_Tranform:Vector=@n10) -> (Output_Get:Vector)
   > Set FACIAL_C_Jaw_Rotation(FACIAL_C_Jaw_Rotation:Rotator=@n10.Rotation) -> (Output_Get:Rotator)
   n11 = BreakTransform(InTransform=GetSocketTransform(self=@n5, InSocketName=FACIAL_C_Jawline, TransformSpace=RTS_ParentBoneSpace))
   > Set FACIAL_C_Jawline_Translation(FACIAL_C_Jawline_Translation:Vector=@n11) -> (Output_Get:Vector)
   > Set FACIAL_C_Jawline_Rotation(FACIAL_C_Jawline_Rotation:Rotator=@n11.Rotation) -> (Output_Get:Rotator)
   n12 = BreakTransform(InTransform=GetSocketTransform(self=@n5, InSocketName=FACIAL_L_Jawline, TransformSpace=RTS_ParentBoneSpace))
   > Set FACIAL_L_Jawline_Translation(FACIAL_L_Jawline_Translation:Vector=@n12) -> (Output_Get:Vector)
   > Set FACIAL_L_Jawline_Rotation(FACIAL_L_Jawline_Rotation:Rotator=@n12.Rotation) -> (Output_Get:Rotator)
   n13 = BreakTransform(InTransform=GetSocketTransform(self=@n5, InSocketName=FACIAL_R_Jawline, TransformSpace=RTS_ParentBoneSpace))
   > Set FACIAL_R_Jawline_Translation(FACIAL_R_Jawline_Translation:Vector=@n13) -> (Output_Get:Vector)
   > Set FACIAL_R_Jawline_Rotation(FACIAL_R_Jawline_Rotation:Rotator=@n13.Rotation) -> (Output_Get:Rotator)
   Outputs!(then:Exec=@n1.then_1)
  ]
 }
]
variables [
 FACIAL_C_Chin_Rotation {type=Rotator}
 FACIAL_C_Chin_Translation {type=Vector}
 FACIAL_C_Jaw_Rotation {type=Rotator}
 FACIAL_C_Jaw_Tranform {type=Vector}
 FACIAL_C_Jawline_Rotation {type=Rotator}
 FACIAL_C_Jawline_Translation {type=Vector}
 FACIAL_C_UnderChin_Rotation {type=Rotator}
 FACIAL_C_UnderChin_Translation {type=Vector}
 FACIAL_L_Jawline_Rotation {type=Rotator}
 FACIAL_L_Jawline_Translation {type=Vector}
 FACIAL_L_UnderChin_Rotation {type=Rotator}
 FACIAL_L_UnderChin_Translation {type=Vector}
 FACIAL_R_Jawline_Rotation {type=Rotator}
 FACIAL_R_Jawline_Translation {type=Vector}
 FACIAL_R_UnderChin_Rotation {type=Rotator}
 FACIAL_R_UnderChin_Translation {type=Vector}
 GravityOverride {type=Vector}
 HeadMesh {type=SkeletalMeshComponentBudgeted}
 isUsingFacialBones {type=boolean}
 Local Noise Offset {type=Vector}
 Local Noise Transform Location {type=Vector}
 Mat Instance Dynamic {type=MaterialInstanceDynamic}
 Overlay Mat Instance Dynamic {type=MaterialInstanceDynamic}
 PlayerPawn {type=FortPlayerPawnAthena}
 R1 Velocity Transform Location {type=Vector}
 R1 Velocity Transform Velocity {type=Vector}
 SkelMesh {type=SkeletalMeshComponent}
 Velocity Interp Speed {type=number}
 WPO Pho Turbulence {type=number}
 WPO Pho Wind Intensity {type=number}
]
