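# Control Rig for the F_MED_AlderOptic face accessory that procedurally poses the character's hair bones around head, shoulders and chest.
# Construction adds reference nulls and controls for both front hair chains.
# Forward solve constrains those controls to spine, clavicle and neck, driving the chains through twist and B-spline IK.
# Kinematic hair bones use weighted constraints plus correctives from swing/twist alphas and spherical pose readers, including one scaled by isIdleMontage.
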ControlRigBlueprint /BRCosmetics/Characters/Player/Female/Medium/Bodies/F_MED_AlderOptic/Meshes/Parts/F_MED_AlderOptic_FaceAcc_CtrlRig
graphs [
 CollapseNode_ContainedGraph {nodes [
   n1 = RigVMFunctionEntryNode()
   hairConstraints = RigVMCollapseNode(ExecutePin=@n1.ExecutePin)
   RigVMFunction_Sequence = RigVMAggregateNode(ExecuteContext=@hairConstraints.ExecutePin)
   GetAlphaFromQuatDelta = RigVMFunctionReferenceNode(
    Rotation Order=ZYX
    child=(Type=Bone,Name="head")
    onlyTwist=true
    outputMinMax=(X=0,Y=1)
    parent=(Type=Null,Name="head_sp_null")
    premult=1
    rotationMinMax=(X=0,Y=-30)
    twistAxis=(X=1,Y=0,Z=0)
    ExecuteContext=@RigVMFunction_Sequence.A
   )
   > &RigUnit_ModifyTransforms(
    ItemToModify=((Item=(Type=Bone,Name="kin_hairFace_l"),Transform=(Rotation=(X=0,Y=0,Z=-0.1736,W=0.9848),Translation=(X=-1.0414,Y=0.47,Z=-2.8967),Scale3D=(X=1,Y=1,Z=1))))
    WeightMinimum=0
    Weight=@GetAlphaFromQuatDelta.alpha
   )
   > GetAlphaFromQuatDelta_1 = RigVMFunctionReferenceNode(
    Rotation Order=ZYX
    child=(Type=Bone,Name="head")
    onlyTwist=true
    outputMinMax=(X=0,Y=1)
    parent=(Type=Null,Name="head_sp_null")
    premult=1
    rotationMinMax=(X=0,Y=30)
    twistAxis=(X=1,Y=0,Z=0)
   )
   > &RigUnit_ModifyTransforms(
    ItemToModify=((Item=(Type=Bone,Name="kin_hairFace_r"),Transform=(Rotation=(X=-0.112,Y=0.1587,Z=0.1849,W=0.9634),Translation=(X=0.2136,Y=-0.7094,Z=-4.6273),Scale3D=(X=1,Y=1,Z=1))))
    WeightMinimum=0
    Weight=@GetAlphaFromQuatDelta_1.alpha
   )
   > &RigUnit_ParentConstraint(
    AdvancedSettings=(RotationOrderForFilter=XZY)
    Child=(Type=Bone,Name="kin_backHair_1")
    Parents="((Item=(Type=Bone,Name=\"upperarm_twist_01_l\"),Weight=0.0121),(Item=(Type=Bone,Name=\"kin_hairFace_l\"),Weight=0.0305),(Item=(Type=Bone,Name=\"neck_02\"),Weight=0.068),(Item=(Type=Bone,Name=\"head\"),Weight=0.1359),(Item=(Type=Bone,Name=\"spine_05\"),Weight=0.3469) ...(truncated)"
    Weight=1
   )
   > &RigUnit_ModifyTransforms(
    ItemToModify=((Item=(Type=Bone,Name="kin_backHair_1"),Transform=(Rotation=(X=0.0449,Y=0.2549,Z=-0.1677,W=0.9513),Translation=(X=1.728,Y=1.3378,Z=-2.7848),Scale3D=(X=1,Y=1,Z=1))))
    Weight=1
    WeightMinimum=0
   )
   ItemArray_4 = RigVMDispatchNode(Value=((Type=Bone,Name="dyn_frontHair_1_l"),(Type=Bone,Name="dyn_frontHair_2_l"),(Type=Bone,Name="dyn_frontHair_3_l"),(Type=Bone,Name="dyn_frontHair_4_l"),(Type=Bone,Name="dyn_frontHair_5_l"))) {template="DISPATCH_RigVMDispatch_Constant(io Value)"}
   TwistBones = RigVMFunctionReferenceNode(
    Axis to Twist=(bX=True,bY=true,bZ=true)
    Debug=false
    End=(Type=Bone,Name="spine_04")
    Start=(Type=Bone,Name="spine_05")
    Bones to Twist=@ItemArray_4.Value
    ExecuteContext=@RigVMFunction_Sequence.B
   )
   ItemArray_5 = RigVMDispatchNode(Value=((Type=Bone,Name="dyn_frontHair_1_r"),(Type=Bone,Name="dyn_frontHair_2_r"),(Type=Bone,Name="dyn_frontHair_3_r"),(Type=Bone,Name="dyn_frontHair_4_r"),(Type=Bone,Name="dyn_frontHair_5_r"))) {template="DISPATCH_RigVMDispatch_Constant(io Value)"}
   > TwistBones_1 = RigVMFunctionReferenceNode(
    Axis to Twist=(bX=True,bY=true,bZ=true)
    Debug=false
    End=(Type=Bone,Name="spine_04")
    Start=(Type=Bone,Name="spine_05")
    Bones to Twist=@ItemArray_5.Value
   )
   ItemArray = RigVMDispatchNode(Value=((Type=Control,Name="dyn_frontHair_1_l_ctrl"),(Type=Control,Name="dyn_frontHair_2_l_ctrl"),(Type=Control,Name="dyn_frontHair_3_l_ctrl"),(Type=Control,Name="dyn_frontHair_4_l_ctrl"),(Type=Control,Name="dyn_frontHair_5_l_ctrl"))) {template="DISPATCH_RigVMDispatch_Constant(io Value)"}
   ItemArray_1 = RigVMDispatchNode(Value=((Type=Bone,Name="dyn_frontHair_1_l"),(Type=Bone,Name="dyn_frontHair_2_l"),(Type=Bone,Name="dyn_frontHair_3_l"),(Type=Bone,Name="dyn_frontHair_4_l"),(Type=Bone,Name="dyn_frontHair_5_l"))) {template="DISPATCH_RigVMDispatch_Constant(io Value)"}
   > SplineIK = RigVMFunctionReferenceNode(
    Primary Axis=(X=1,Y=0,Z=0)
    Samples Per Segment=2
    Secondary Spline Direction=(X=0,Y=0,Z=0)
    Spline Mode=BSpline
    Up Axis=(X=0,Y=0,Z=0)
    Controls=@ItemArray.Value
    Bones=@ItemArray_1.Value
   )
   ItemArray_2 = RigVMDispatchNode(Value=((Type=Control,Name="dyn_frontHair_1_r_ctrl"),(Type=Control,Name="dyn_frontHair_2_r_ctrl"),(Type=Control,Name="dyn_frontHair_3_r_ctrl"),(Type=Control,Name="dyn_frontHair_4_r_ctrl"),(Type=Control,Name="dyn_frontHair_5_r_ctrl"))) {template="DISPATCH_RigVMDispatch_Constant(io Value)"}
   ItemArray_3 = RigVMDispatchNode(Value=((Type=Bone,Name="dyn_frontHair_1_r"),(Type=Bone,Name="dyn_frontHair_2_r"),(Type=Bone,Name="dyn_frontHair_3_r"),(Type=Bone,Name="dyn_frontHair_4_r"),(Type=Bone,Name="dyn_frontHair_5_r"))) {template="DISPATCH_RigVMDispatch_Constant(io Value)"}
   > SplineIK_1 = RigVMFunctionReferenceNode(
    Primary Axis=(X=1,Y=0,Z=0)
    Samples Per Segment=2
    Secondary Spline Direction=(X=0,Y=0,Z=0)
    Spline Mode=BSpline
    Up Axis=(X=0,Y=0,Z=0)
    Controls=@ItemArray_2.Value
    Bones=@ItemArray_3.Value
   )
   n2 = &RigUnit_SphericalPoseReader(
    ActiveRegionScaleFactors=(PositiveWidth=1.2,NegativeWidth=2,PositiveHeight=2,NegativeHeight=2)
    ActiveRegionSize=0.06
    Debug=(bDrawDebug=false,bDraw2D=False,bDrawLocalAxes=False,DebugScale=25,DebugSegments=20,DebugThickness=0.25)
    DriverAxis=(X=-1,Y=0,Z=0)
    DriverItem=(Type=Bone,Name="clavicle_r")
    FalloffRegionScaleFactors=(PositiveWidth=1,NegativeWidth=1,PositiveHeight=1,NegativeHeight=1)
    FalloffSize=0.18
    FlipHeightScaling=False
    FlipWidthScaling=False
    OptionalParentItem=(Type=Bone,Name="None")
    RotationOffset=(X=-10,Y=120,Z=0)
    ExecutePin=@RigVMFunction_Sequence.C
   )
   > n3 = &RigUnit_SphericalPoseReader(
    ActiveRegionScaleFactors=(PositiveWidth=4,NegativeWidth=4,PositiveHeight=1,NegativeHeight=4)
    ActiveRegionSize=0.03
    Debug=(bDrawDebug=false,bDraw2D=False,bDrawLocalAxes=False,DebugScale=25,DebugSegments=20,DebugThickness=0.25)
    DriverAxis=(X=-1,Y=0,Z=0)
    DriverItem=(Type=Bone,Name="clavicle_r")
    FalloffRegionScaleFactors=(PositiveWidth=1,NegativeWidth=1,PositiveHeight=0.8,NegativeHeight=1)
    FalloffSize=0.15
    FlipHeightScaling=False
    FlipWidthScaling=False
    OptionalParentItem=(Type=Bone,Name="None")
    RotationOffset=(X=-30,Y=100,Z=0)
   )
   Maximum = &RigVMFunction_MathDoubleMax(A=@n3.OutputParam, B=@n2.OutputParam)
   > &RigUnit_ModifyTransforms(
    ItemToModify="((Item=(Type=Bone,Name=\"kin_frontHair_1_r\"),Transform=(Rotation=(X=0,Y=0,Z=0,W=1),Translation=(X=-0.4992,Y=-1.641,Z=1.8238),Scale3D=(X=1,Y=1,Z=1))),(Item=(Type=Bone,Name=\"kin_frontHair_2_r\"),Transform=(Rotation=(X=0,Y=0,Z=0,W=1),Translation=(X=0,Y=-1.9083, ...(truncated)"
    WeightMinimum=0
    Weight=@Maximum.Result
   )
   > &RigUnit_ParentConstraint(
    AdvancedSettings=(RotationOrderForFilter=XZY)
    Child=(Type=Bone,Name="kin_sideHair_1_l")
    Parents="((Item=(Type=Bone,Name=\"neck_02\"),Weight=0.0167),(Item=(Type=Bone,Name=\"kin_hairFace_l\"),Weight=0.0199),(Item=(Type=Bone,Name=\"upperarm_twist_01_l\"),Weight=0.0578),(Item=(Type=Bone,Name=\"kin_frontHair_1_l\"),Weight=0.0794),(Item=(Type=Bone,Name=\"head\"),Weig ...(truncated)"
    Weight=1
   )
   > GetAlphaFromQuatDelta_2 = RigVMFunctionReferenceNode(
    Rotation Order=ZYX
    child=(Type=Bone,Name="clavicle_l")
    outputMinMax=(X=0,Y=1)
    parent=(Type=Null,Name="clavicle_l_null")
    premult=1
    rotationMinMax=(X=0,Y=-30)
    triggerAxis=Y
    twistAxis=(X=1,Y=0,Z=0)
   )
   > &RigUnit_ModifyTransforms(
    ItemToModify="((Item=(Type=Bone,Name=\"kin_sideHair_1_l\"),Transform=(Rotation=(X=0,Y=-0.1329,Z=0,W=0.9911),Translation=(X=-0.0492,Y=0,Z=0.5263),Scale3D=(X=1,Y=1,Z=1))),(Item=(Type=Bone,Name=\"kin_frontHair_1_l\"),Transform=(Rotation=(X=0,Y=-0.0954,Z=0,W=0.9954),Translation ...(truncated)"
    WeightMinimum=0
    Weight=@GetAlphaFromQuatDelta_2.alpha
   )
   > n4 = &RigUnit_SphericalPoseReader(
    ActiveRegionScaleFactors=(PositiveWidth=1,NegativeWidth=1,PositiveHeight=1,NegativeHeight=1)
    ActiveRegionSize=0.1
    Debug=(bDrawDebug=false,bDraw2D=False,bDrawLocalAxes=False,DebugScale=25,DebugSegments=20,DebugThickness=0.1)
    DriverAxis=(X=1,Y=0,Z=0)
    DriverItem=(Type=Bone,Name="clavicle_l")
    FalloffRegionScaleFactors=(PositiveWidth=1,NegativeWidth=1,PositiveHeight=1,NegativeHeight=1)
    FalloffSize=0.2
    FlipHeightScaling=False
    FlipWidthScaling=False
    OptionalParentItem=(Type=Bone,Name="None")
    RotationOffset=(X=35,Y=-90,Z=0)
   )
   > &RigUnit_ModifyTransforms(
    ItemToModify="((Item=(Type=Bone,Name=\"kin_sideHair_1_l\"),Transform=(Rotation=(X=0,Y=-0.0698,Z=0,W=0.9976),Translation=(X=0,Y=0,Z=1),Scale3D=(X=1,Y=1,Z=1))),(Item=(Type=Bone,Name=\"kin_frontHair_1_l\"),Transform=(Rotation=(X=0,Y=0,Z=0,W=-1),Translation=(X=0,Y=0,Z=1.0735),S ...(truncated)"
    WeightMinimum=0
    Weight=@n4.OutputParam
   )
   > n5 = &RigUnit_SphericalPoseReader(
    ActiveRegionScaleFactors=(PositiveWidth=2,NegativeWidth=1,PositiveHeight=2,NegativeHeight=1)
    ActiveRegionSize=0.04
    Debug=(bDrawDebug=false,bDraw2D=False,bDrawLocalAxes=False,DebugScale=25,DebugSegments=20,DebugThickness=0.25)
    DriverAxis=(X=1,Y=0,Z=0)
    DriverItem=(Type=Bone,Name="head")
    FalloffRegionScaleFactors=(PositiveWidth=1,NegativeWidth=0.3,PositiveHeight=1,NegativeHeight=1)
    FalloffSize=0.2
    FlipHeightScaling=False
    FlipWidthScaling=False
    OptionalParentItem=(Type=Bone,Name="None")
    RotationOffset=(X=5,Y=-115,Z=8)
   )
   Multiply = &RigVMFunction_MathDoubleMul(A=@n5.OutputParam, B=RigVMVariableNode() {variable=isIdleMontage}.Value)
   > &RigUnit_ModifyTransforms(
    ItemToModify=((Item=(Type=Bone,Name="kin_hairFace_l"),Transform=(Rotation=(X=0.1478,Y=0,Z=0,W=0.989),Translation=(X=2.5,Y=1.5,Z=0),Scale3D=(X=0.4,Y=1,Z=1))))
    WeightMinimum=0
    Weight=@Multiply.Result
   )
   RigVMFunctionReturnNode(ExecutePin=@RigVMFunction_Sequence.D)
  ]}
 CollapseNode_ContainedGraph {nodes [
   n1 = RigVMFunctionEntryNode()
   &RigUnit_ParentConstraint(
    AdvancedSettings=(RotationOrderForFilter=XZY)
    Child=(Type=Null,Name="dyn_frontHair_1_l_ctrl_null")
    Filter=(TranslationFilter=(bX=True,bY=True,bZ=True),RotationFilter=(bX=True,bY=True,bZ=True),ScaleFilter=(bX=False,bY=False,bZ=False))
    Parents=((Item=(Type=Bone,Name="clavicle_l"),Weight=0.4479),(Item=(Type=Bone,Name="spine_05"),Weight=0.5021),(Item=(Type=Bone,Name="neck_01"),Weight=0.05))
    Weight=1
    ExecutePin=@n1.ExecutePin
   )
   > &RigUnit_ParentConstraint(
    AdvancedSettings=(RotationOrderForFilter=XZY)
    Child=(Type=Null,Name="dyn_frontHair_2_l_ctrl_null")
    Filter=(TranslationFilter=(bX=True,bY=True,bZ=True),RotationFilter=(bX=True,bY=True,bZ=True),ScaleFilter=(bX=False,bY=False,bZ=False))
    Parents=((Item=(Type=Bone,Name="clavicle_l"),Weight=0.2453),(Item=(Type=Bone,Name="spine_05"),Weight=0.7547))
    Weight=1
   )
   > &RigUnit_ParentConstraint(
    AdvancedSettings=(RotationOrderForFilter=XZY)
    Child=(Type=Null,Name="dyn_frontHair_3_l_ctrl_null")
    Filter=(TranslationFilter=(bX=True,bY=True,bZ=True),RotationFilter=(bX=True,bY=True,bZ=True),ScaleFilter=(bX=False,bY=False,bZ=False))
    Parents=((Item=(Type=Bone,Name="spine_04"),Weight=0.0331),(Item=(Type=Bone,Name="clavicle_l"),Weight=0.1387),(Item=(Type=Bone,Name="spine_05"),Weight=0.8282))
    Weight=1
   )
   > &RigUnit_ParentConstraint(
    AdvancedSettings=(RotationOrderForFilter=XZY)
    Child=(Type=Null,Name="dyn_frontHair_4_l_ctrl_null")
    Filter=(TranslationFilter=(bX=True,bY=True,bZ=True),RotationFilter=(bX=True,bY=True,bZ=True),ScaleFilter=(bX=False,bY=False,bZ=False))
    Parents=((Item=(Type=Bone,Name="clavicle_l"),Weight=0.0139),(Item=(Type=Bone,Name="spine_04"),Weight=0.1148),(Item=(Type=Bone,Name="spine_05"),Weight=0.8713))
    Weight=1
   )
   > &RigUnit_ParentConstraint(
    AdvancedSettings=(RotationOrderForFilter=XZY)
    Child=(Type=Null,Name="dyn_frontHair_5_l_ctrl_null")
    Filter=(TranslationFilter=(bX=True,bY=True,bZ=True),RotationFilter=(bX=True,bY=True,bZ=True),ScaleFilter=(bX=False,bY=False,bZ=False))
    Parents=((Item=(Type=Bone,Name="spine_05"),Weight=0.4117),(Item=(Type=Bone,Name="spine_04"),Weight=0.5883))
    Weight=1
   )
   > &RigUnit_ParentConstraint(
    AdvancedSettings=(RotationOrderForFilter=XZY)
    Child=(Type=Null,Name="dyn_frontHair_1_r_ctrl_null")
    Filter=(TranslationFilter=(bX=True,bY=True,bZ=True),RotationFilter=(bX=True,bY=True,bZ=True),ScaleFilter=(bX=False,bY=False,bZ=False))
    Parents=((Item=(Type=Bone,Name="clavicle_r"),Weight=0.0428),(Item=(Type=Bone,Name="neck_02"),Weight=0.0567),(Item=(Type=Bone,Name="neck_01"),Weight=0.1014),(Item=(Type=Bone,Name="spine_05"),Weight=0.7991))
    Weight=1
    bMaintainOffset=true
   )
   > &RigUnit_ParentConstraint(
    AdvancedSettings=(RotationOrderForFilter=XZY)
    Child=(Type=Null,Name="dyn_frontHair_2_r_ctrl_null")
    Filter=(TranslationFilter=(bX=True,bY=True,bZ=True),RotationFilter=(bX=True,bY=True,bZ=True),ScaleFilter=(bX=False,bY=False,bZ=False))
    Parents=((Item=(Type=Bone,Name="clavicle_r"),Weight=0.1029),(Item=(Type=Bone,Name="spine_05"),Weight=0.8971))
    Weight=1
   )
   > &RigUnit_ParentConstraint(
    AdvancedSettings=(RotationOrderForFilter=XZY)
    Child=(Type=Null,Name="dyn_frontHair_3_r_ctrl_null")
    Filter=(TranslationFilter=(bX=True,bY=True,bZ=True),RotationFilter=(bX=True,bY=True,bZ=True),ScaleFilter=(bX=False,bY=False,bZ=False))
    Parents=((Item=(Type=Bone,Name="clavicle_r"),Weight=0.0741),(Item=(Type=Bone,Name="spine_04"),Weight=0.3055),(Item=(Type=Bone,Name="spine_05"),Weight=0.6203))
    Weight=1
   )
   > &RigUnit_ParentConstraint(
    AdvancedSettings=(RotationOrderForFilter=XZY)
    Child=(Type=Null,Name="dyn_frontHair_4_r_ctrl_null")
    Filter=(TranslationFilter=(bX=True,bY=True,bZ=True),RotationFilter=(bX=True,bY=True,bZ=True),ScaleFilter=(bX=False,bY=False,bZ=False))
    Parents=((Item=(Type=Bone,Name="clavicle_r"),Weight=0.0217),(Item=(Type=Bone,Name="spine_05"),Weight=0.4177),(Item=(Type=Bone,Name="spine_04"),Weight=0.5606))
    Weight=1
   )
   > n2 = &RigUnit_ParentConstraint(
    AdvancedSettings=(RotationOrderForFilter=XZY)
    Child=(Type=Null,Name="dyn_frontHair_5_r_ctrl_null")
    Filter=(TranslationFilter=(bX=True,bY=True,bZ=True),RotationFilter=(bX=True,bY=True,bZ=True),ScaleFilter=(bX=False,bY=False,bZ=False))
    Parents=((Item=(Type=Bone,Name="spine_05"),Weight=0.1753),(Item=(Type=Bone,Name="spine_04"),Weight=0.8247))
    Weight=1
   )
   RigVMFunctionReturnNode(ExecutePin=@n2.ExecutePin)
  ]}
 New Function_ContainedGraph {nodes [
   RigVMFunctionEntryNode()
   RigVMFunction_MathBoolToInteger = &RigVMFunction_MathBoolToInteger(Value=RigVMVariableNode() {variable=onlyTwist}.Value)
   RigVMFunction_MathQuaternionSwingTwist = &RigVMFunction_MathQuaternionSwingTwist(
    Input=&RigUnit_GetRelativeTransformForItem(
     bChildInitial=False
     bParentInitial=False
     Child=RigVMVariableNode() {variable=child}.Value
     Parent=RigVMVariableNode() {variable=parent}.Value
    ).RelativeTransform.Rotation
    TwistAxis=RigVMVariableNode() {variable=twistAxis}.Value
   )
   n1 = RigVMDispatchNode(
    Values=((),())
    Index=@RigVMFunction_MathBoolToInteger.Result
    Values.0=@RigVMFunction_MathQuaternionSwingTwist.Swing
    Values.1=@RigVMFunction_MathQuaternionSwingTwist.Twist
   ) {template="DISPATCH_RigVMDispatch_SelectInt32(in Index,in Values,out Result)"}
   > RigVMVariableNode(Value=@n1.Result) {variable=entry_outputQuat}
   RigVMFunction_MathQuaternionToEuler = &RigVMFunction_MathQuaternionToEuler(Value=@n1.Result, RotationOrder=RigVMVariableNode() {variable=Rotation Order}.Value)
   > DISPATCH_RigVMDispatch_Print = RigVMDispatchNode(
    ScreenColor=(R=1,G=1,B=1,A=1)
    ScreenDuration=1.5
    Value=@RigVMFunction_MathQuaternionToEuler.Result
    Enabled=RigVMVariableNode() {variable=debugRotation}.Value
   ) {template="DISPATCH_RigVMDispatch_Print(in Prefix,in Value,in Enabled,in ScreenDuration,in ScreenColor)"}
   Cast = RigVMDispatchNode(Value=RigVMVariableNode() {variable=triggerAxis}.Value) {template="DISPATCH_RigVMDispatch_CastEnumToInt(in Value,out Result)"}
   DISPATCH_RigVMDispatch_SelectInt32 = RigVMDispatchNode(
    Values=(0,0,0)
    Index=@Cast.Result
    Values.0=@RigVMFunction_MathQuaternionToEuler.Result.X
    Values.1=@RigVMFunction_MathQuaternionToEuler.Result.Y
    Values.2=@RigVMFunction_MathQuaternionToEuler.Result.Z
   ) {template="DISPATCH_RigVMDispatch_SelectInt32(in Index,in Values,out Result)"}
   n2 = RigVMVariableNode() {variable=rotationMinMax}
   n3 = RigVMVariableNode() {variable=outputMinMax}
   Remap = &RigVMFunction_MathDoubleRemap(
    bClamp=False
    Value=@DISPATCH_RigVMDispatch_SelectInt32.Result
    SourceMinimum=@n2.Value.X
    SourceMaximum=@n2.Value.Y
    TargetMinimum=@n3.Value.X
    TargetMaximum=@n3.Value.Y
   )
   Absolute = &RigVMFunction_MathDoubleAbs(Value=@Remap.Result)
   n4 = RigVMVariableNode() {variable=absoluteValue}
   Multiply = &RigVMFunction_MathDoubleMul(
    A=&RigVMFunction_MathDoubleClamp(Value=RigVMDispatchNode(True=@Absolute.Result, False=@Remap.Result, Condition=@n4.Value) {template="DISPATCH_RigVMDispatch_If(in Condition,in True,in False,out Result)"}.Result).Result
    B=RigVMVariableNode() {variable=premult}.Value
   )
   > RigVMFunctionReturnNode(alpha=@Multiply.Result, rawQuat=RigVMVariableNode() {variable=entry_outputQuat}.Value)
  ]}
 RigVMFunctionLibrary {nodes [GetAlphaFromQuatDelta = RigVMCollapseNode(
    ExecuteContext=()
    Rotation Order=ZYX
    child=(Type=Bone,Name="None")
    outputMinMax=(X=0,Y=1)
    parent=(Type=Bone,Name="None")
    premult=1
    rotationMinMax=(X=0,Y=0)
    twistAxis=(X=1,Y=0,Z=0)
   )]}
 RigVMFunction_Sequence_1_ContainedGraph {nodes [
   n1 = RigVMFunctionEntryNode()
   RigVMFunction_Sequence = &RigVMFunction_Sequence(ExecuteContext=@n1.ExecuteContext)
   n2 = &RigVMFunction_Sequence(ExecuteContext=@RigVMFunction_Sequence.B)
   n3 = &RigVMFunction_Sequence(ExecuteContext=@n2.B)
   RigVMFunctionReturnNode(A=@RigVMFunction_Sequence.A, B=@n2.A, C=@n3.A, D=@n3.B)
  ]}
 RigVMFunction_Sequence_1_ContainedGraph {nodes [
   n1 = RigVMFunctionEntryNode()
   RigVMFunction_Sequence = &RigVMFunction_Sequence(ExecuteContext=@n1.ExecuteContext)
   n2 = &RigVMFunction_Sequence(ExecuteContext=@RigVMFunction_Sequence.B)
   RigVMFunctionReturnNode(A=@RigVMFunction_Sequence.A, B=@n2.A, C=@n2.B)
  ]}
 RigVMModel {nodes [
   RigUnit_BeginExecution = &RigUnit_BeginExecution()
   hairSetup = RigVMCollapseNode(ExecutePin=@RigUnit_BeginExecution.ExecutePin)
  ]}
 RigVMModel Rig {nodes [
   n1 = &RigUnit_PrepareForExecution()
   RigVMFunction_Sequence = RigVMAggregateNode(ExecuteContext=@n1.ExecutePin)
   ItemArray_2 = RigVMDispatchNode(Value=((Type=Bone,Name="clavicle_l"))) {template="DISPATCH_RigVMDispatch_Constant(io Value)"}
   DISPATCH_RigVMDispatch_ArrayIterator = RigVMDispatchNode(Array=@ItemArray_2.Value, ExecuteContext=@RigVMFunction_Sequence.A) {template="DISPATCH_RigVMDispatch_ArrayIterator(in Array,out Element,out Index,out Count,out Ratio)"}
   Concat = /Script/RigVM.RigVMFunction_NameConcat(B=_null, A=@DISPATCH_RigVMDispatch_ArrayIterator.Element.Name) {is=struct RigVMUnitNode, template="Concat::Execute(in A,in B,out Result)"}
   > &RigUnit_HierarchyAddNull(
    Space=LocalSpace
    Name=@Concat.Result
    Transform=&RigUnit_GetTransform(Item=(Type=Bone,Name="clavicle_l"), Space=LocalSpace, bInitial=true, Item.Name=@DISPATCH_RigVMDispatch_ArrayIterator.Element.Name).Transform
    Parent=/Script/ControlRig.RigUnit_HierarchyGetParent(bDefaultParent=True, Child=@DISPATCH_RigVMDispatch_ArrayIterator.Element) {is=struct RigVMUnitNode}.Parent
   )
   &RigUnit_HierarchyAddNull(
    Name=head_sp_null
    Parent=(Type=Bone,Name="spine_05")
    Space=LocalSpace
    Transform=&RigUnit_GetRelativeTransformForItem(Child=(Type=Bone,Name="head"), Parent=(Type=Bone,Name="spine_05"), bChildInitial=true, bParentInitial=true).RelativeTransform
    ExecutePin=@RigVMFunction_Sequence.B
   )
   ItemArray = RigVMDispatchNode(Value=((Type=Bone,Name="dyn_frontHair_1_l"),(Type=Bone,Name="dyn_frontHair_2_l"),(Type=Bone,Name="dyn_frontHair_3_l"),(Type=Bone,Name="dyn_frontHair_4_l"),(Type=Bone,Name="dyn_frontHair_5_l"))) {template="DISPATCH_RigVMDispatch_Constant(io Value)"}
   For_Each = RigVMDispatchNode(Array=@ItemArray.Value, ExecuteContext=@RigVMFunction_Sequence.C) {template="DISPATCH_RigVMDispatch_ArrayIterator(in Array,out Element,out Index,out Count,out Ratio)"}
   RigUnit_GetTransform = &RigUnit_GetTransform(bInitial=true, Item=@For_Each.Element)
   > Control Stack at Position = RigVMFunctionReferenceNode(
    Add Bottom Null=False
    Add Null=true
    Add Secondary Control=False
    Control Suffix=_ctrl
    Null Suffix=_ctrl_null
    Parent=(Type=None,Name="None")
    Secondary Settings="(Shape=(bVisible=True,Name=\"Default\",Color=(R=1,G=0,B=0,A=1),Transform=(Rotation=(X=0,Y=0,Z=0,W=1),Translation=(X=0,Y=0,Z=0),Scale3D=(X=1,Y=1,Z=1))),Proxy=(bIsProxy=False,ShapeVisibility=BasedOnSelection),DisplayName=\"None\",InitialSpace=LocalSpace,bUsePref ...(truncated)"
    Settings="(Shape=(bVisible=True,Name=\"Circle_Thin\",Color=(R=0,G=0,B=1,A=1),Transform=(Rotation=(X=0,Y=-0.7071,Z=0,W=0.7071),Translation=(X=0,Y=0,Z=0),Scale3D=(X=0.6,Y=0.6,Z=0.6))),Proxy=(bIsProxy=False,ShapeVisibility=BasedOnSelection),DisplayName=\"None\",InitialSpac ...(truncated)"
    Name=@For_Each.Element.Name
    Global Position=@RigUnit_GetTransform.Transform
   )
   ItemArray_1 = RigVMDispatchNode(Value=((Type=Bone,Name="dyn_frontHair_1_r"),(Type=Bone,Name="dyn_frontHair_2_r"),(Type=Bone,Name="dyn_frontHair_3_r"),(Type=Bone,Name="dyn_frontHair_4_r"),(Type=Bone,Name="dyn_frontHair_5_r"))) {template="DISPATCH_RigVMDispatch_Constant(io Value)"}
   For_Each_1 = RigVMDispatchNode(ExecuteContext=@For_Each.Completed, Array=@ItemArray_1.Value) {template="DISPATCH_RigVMDispatch_ArrayIterator(in Array,out Element,out Index,out Count,out Ratio)"}
   > Control Stack at Position_1 = RigVMFunctionReferenceNode(
    Add Bottom Null=False
    Add Null=true
    Add Secondary Control=False
    Control Suffix=_ctrl
    Null Suffix=_ctrl_null
    Parent=(Type=None,Name="None")
    Secondary Settings="(Shape=(bVisible=True,Name=\"Default\",Color=(R=1,G=0,B=0,A=1),Transform=(Rotation=(X=0,Y=0,Z=0,W=1),Translation=(X=0,Y=0,Z=0),Scale3D=(X=1,Y=1,Z=1))),Proxy=(bIsProxy=False,ShapeVisibility=BasedOnSelection),DisplayName=\"None\",InitialSpace=LocalSpace,bUsePref ...(truncated)"
    Settings="(Shape=(bVisible=True,Name=\"Circle_Thin\",Color=(R=1,G=0,B=0,A=1),Transform=(Rotation=(X=0,Y=-0.7071,Z=0,W=0.7071),Translation=(X=0,Y=0,Z=0),Scale3D=(X=0.6,Y=0.6,Z=0.6))),Proxy=(bIsProxy=False,ShapeVisibility=BasedOnSelection),DisplayName=\"None\",InitialSpac ...(truncated)"
    Name=@For_Each_1.Element.Name
    Global Position=&RigUnit_GetTransform(bInitial=true, Item=@For_Each_1.Element).Transform
   )
  ]}
]
hierarchy [
 root {loc=<0 0 0>, rot=<0 0 0 1>}
 pelvis {parent=root, loc=<0 -0.5301 98.0571>, rot=<0.0127 0.707 -0.0127 -0.707>}
 spine_01 {parent=pelvis, loc=<6.1208 -0.9328 0>, rot=<0 0 0.034 -0.9994>}
 spine_02 {parent=spine_01, loc=<6.5997 -0.2095 0>, rot=<0 0 0.0238 -0.9997>}
 spine_03 {parent=spine_02, loc=<6.4914 -0.1237 0>, rot=<0 0 0.0149 -0.9999>}
 spine_04 {parent=spine_03, loc=<6.9598 0.0794 0>, rot=<0 0 -0.0132 -0.9999>}
 spine_05 {parent=spine_04, loc=<6.9987 0 0>, rot=<0 0 -0.0015 -1>}
 neck_01 {parent=spine_05, loc=<11.3138 0.1529 0>, rot=<0 0 -0.1994 -0.9799>}
 neck_02 {parent=neck_01, loc=<4.5627 0 0>, rot=<0 0 0.0504 -0.9987>}
 head {parent=neck_02, loc=<5.3914 0 0>, rot=<0 0 0.1132 -0.9936>}
 kin_hairFace_l {parent=head, loc=<0 0 -0.25>, rot=<0.0023 -0.7071 -0.0023 -0.7071>}
 kin_hairFace_r {parent=head, loc=<0 0 0.25>, rot=<0.0023 -0.7071 -0.0023 -0.7071>}
 kin_backHair_1 {parent=head, loc=<-6.6487 -4.2206 -9.8463>, rot=<-0.2399 -0.9426 -0.0587 -0.2245>}
 clavicle_l {parent=spine_05, loc=<5.4831 -0.1053 -3.2946>, rot=<0.0465 -0.7289 -0.009 -0.6829>}
 upperarm_l {parent=clavicle_l, loc=<9.8658 -1.2136 -0.4105>, rot=<0.0164 -0.3245 -0.0269 -0.9454>}
 upperarm_twist_01_l {parent=upperarm_l, loc=<7.2135 -0.1134 0>, rot=<0 0 0 1>}
 clavicle_r {parent=spine_05, loc=<5.4831 -0.1053 3.2946>, rot=<0.7289 0.0465 -0.6829 0.009>}
 upperarm_r {parent=clavicle_r, loc=<-9.8658 1.2136 0.4105>, rot=<0.0164 -0.3245 -0.0269 -0.9454>}
 upperarm_twist_01_r {parent=upperarm_r, loc=<-7.2135 0.1134 0>, rot=<0 0 0 -1>}
 coll_chest_01_r {parent=spine_05, loc=<-1.2647 9.6927 2.2211>, rot=<0.6168 0.1089 0.4096 -0.6633>}
 coll_chest_01_l {parent=spine_05, loc=<-1.7378 11.8051 -1.6885>, rot=<0.5599 -0.4551 0.0255 -0.6919>}
 base_frontHair_1_l {parent=spine_05, loc=<10.8261 5.574 -7.3051>, rot=<0.2572 0.6727 0.6854 0.1073>}
 dyn_frontHair_1_l {parent=base_frontHair_1_l, loc=<0 0 0>, rot=<0 0 0 1>}
 dyn_frontHair_2_l {parent=dyn_frontHair_1_l, loc=<3.6945 0 0>, rot=<0 0 -0.0999 -0.995>}
 dyn_frontHair_3_l {parent=dyn_frontHair_2_l, loc=<3.6604 0 0>, rot=<0.0509 -0.0077 -0.1488 -0.9875>}
 dyn_frontHair_4_l {parent=dyn_frontHair_3_l, loc=<3.7248 0 0>, rot=<0.0293 0.0795 0.2392 -0.9673>}
 dyn_frontHair_5_l {parent=dyn_frontHair_4_l, loc=<3.5469 0 0>, rot=<0 0 0 1>}
 kin_frontHair_1_l {parent=base_frontHair_1_l, loc=<0 -3.0985 -1.9868>, rot=<0 0 0 1>}
 base_frontHair_1_r {parent=spine_05, loc=<10.0338 6.5477 5.646>, rot=<0.2576 0.5737 0.7596 0.1659>}
 dyn_frontHair_1_r {parent=base_frontHair_1_r, loc=<0 0 0>, rot=<0 0 0 1>}
 dyn_frontHair_2_r {parent=dyn_frontHair_1_r, loc=<4.3119 0 0>, rot=<0 -0.0863 0.0231 -0.996>}
 dyn_frontHair_3_r {parent=dyn_frontHair_2_r, loc=<4.2073 0 0>, rot=<-0.0459 -0.1544 0.2401 -0.9573>}
 dyn_frontHair_4_r {parent=dyn_frontHair_3_r, loc=<4.3301 0 0>, rot=<0.1106 0.0498 0.1687 -0.9782>}
 dyn_frontHair_5_r {parent=dyn_frontHair_4_r, loc=<4.2987 0 0>, rot=<0 0 0 1>}
 kin_frontHair_2_r {parent=dyn_frontHair_2_r, loc=<0 3.3855 -1.8453>, rot=<0 0 0 1>}
 kin_frontHair_1_r {parent=dyn_frontHair_1_r, loc=<0 4.0387 -1.0643>, rot=<0 0 0 1>}
 kin_sideHair_1_l {parent=spine_05, loc=<11.2815 2.4795 -10.0225>, rot=<0.3625 0.5718 0.7118 0.187>}
 coll_breast_1_r {parent=spine_05, loc=<-4.8281 9.3971 8.6449>, rot=<0.0004 -0.7021 -0.0567 -0.7099>}
]
variables [isIdleMontage {type=double, default=0}]
