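# Physics asset for the F_MED_AlderOptic female face accessory mesh, setting up physics for two front hair chains, left and right.
# Each chain links four capsule dyn_frontHair bones in sequence from a base bone, with tight swing and twist limits, slightly wider on the lower left links.
# Unconstrained capsules on the clavicles, chest, upper spine and right breast give the hair torso shapes to collide with.
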PhysicsAsset /BRCosmetics/Characters/Player/Female/Medium/Bodies/F_MED_AlderOptic/Meshes/Parts/F_MED_AlderOptic_FaceAcc_Physics
bodies {
 base_frontHair_1_l [{shapeType=Capsule, length=0.505, radius=0.505, shapeName=base_frontHair_1_l_capsule, rotation=<rot 0 -90 0>}]
 base_frontHair_1_r [{shapeType=Capsule, length=0.505, radius=0.505, shapeName=base_frontHair_1_r_capsule, rotation=<rot 0 -90 0>}]
 clavicle_l [{shapeType=Capsule, length=7.3965, radius=5.4153, shapeName=clavicle_l_capsule, center=<6.9998 0 -0.4067>, rotation=<rot 90 0 0>}]
 clavicle_r [{shapeType=Capsule, length=7.3965, radius=5.4153, shapeName=clavicle_r_capsule, center=<-6.9998 0 0.4067>, rotation=<rot 90 180 0>}]
 coll_breast_1_r [{shapeType=Capsule, length=5.6756, radius=2.8354, shapeName=coll_breast_1_r_capsule, center=<3.3916 0.1344 0>, rotation=<rot 90 0 0>}]
 coll_chest_01_l [{shapeType=Capsule, length=9.8964, radius=2.4207, shapeName=coll_chest_01_l_capsule, center=<5.7559 0 -0.8516>, rotation=<rot 90 0 0>}]
 coll_chest_01_r [{
   shapeType=Capsule
   length=9.0749
   radius=1.9506
   shapeName=coll_chest_01_l_capsule
   center=<6.361 0 -0.618>
   rotation=<rot 89.9052 3.8253 3.8316>
  }]
 dyn_frontHair_1_l [{
   shapeType=Capsule
   length=4.4682
   radius=2.3422
   shapeName=dyn_frontHair_1_l_capsule
   center=<0.9815 -0.4392 -0.2946>
   rotation=<rot 0 157.5538 -65.0254>
  }]
 dyn_frontHair_1_r [{
   shapeType=Capsule
   length=4.9838
   radius=2.284
   shapeName=dyn_frontHair_1_r_capsule
   center=<1.6367 0.3155 0.5683>
   rotation=<rot 0 3.4149 -86.9943>
  }]
 dyn_frontHair_2_l [{
   shapeType=Capsule
   length=1.5736
   radius=1.6353
   shapeName=dyn_frontHair_2_l_capsule
   center=<2.2241 0.0585 0.2903>
   rotation=<rot -84.1319 6.8003 0.0006>
  }]
 dyn_frontHair_2_r [{
   shapeType=Capsule
   length=2.7838
   radius=2.8628
   shapeName=dyn_frontHair_2_r_capsule
   center=<1.895 1.0356 1.093>
   rotation=<rot -5.6626 -0.1684 -83.5859>
  }]
 dyn_frontHair_3_l [{
   shapeType=Capsule
   length=1.6241
   radius=1.2464
   shapeName=dyn_frontHair_3_l_capsule
   center=<1.5975 0.2203 0.3157>
   rotation=<rot -82.3582 -0.0002 0.0002>
  }]
 dyn_frontHair_3_r [
  {
   shapeType=Capsule
   length=3.0975
   radius=2.5576
   shapeName=dyn_frontHair_3_r_capsule
   center=<1.9422 -0.1997 0.7762>
   rotation=<rot 0.0001 70.7883 -77.8546>
  }
  {
   shapeType=Capsule
   length=2.296
   radius=0.9807
   shapeName=dyn_frontHair_3_r_capsule
   center=<2.7915 1.2407 -0.748>
   rotation=<rot 0.0001 70.7883 -72.7333>
  }
 ]
 dyn_frontHair_4_l [{
   shapeType=Capsule
   length=2.6538
   radius=1.0106
   shapeName=dyn_frontHair_4_l_capsule
   center=<2.1408 0.4918 0.3499>
   rotation=<rot -0.1723 -92.1953 -92.6397>
  }]
 dyn_frontHair_4_r [
  {
   shapeType=Capsule
   length=2.8182
   radius=0.9265
   shapeName=dyn_frontHair_4_r_capsule
   center=<2.3969 -0.0591 0.333>
   rotation=<rot 15.4781 94.2461 -83.4337>
  }
  {
   shapeType=Capsule
   length=1.9319
   radius=0.5
   shapeName=dyn_frontHair_4_r_capsule
   center=<1.9497 1.0858 0.6758>
   rotation=<rot 15.4781 94.2461 -77.8123>
  }
 ]
 spine_04 [
  {shapeType=Capsule, length=11.2494, radius=9.1645, shapeName=spine_04_capsule, center=<0.4658 2.8214 0>}
  {shapeType=Capsule, length=13.5803, radius=6.8336, shapeName=spine_04_capsule, center=<-0.304 6.3189 0>}
 ]
 spine_05 [{shapeType=Capsule, length=11.2494, radius=9.1645, shapeName=spine_04_capsule, center=<0 -0.5054 0>}]
}
constraints [
 {bone1Name=dyn_frontHair_1_l, bone2Name=base_frontHair_1_l, swing1LimitDegrees=20, swing2LimitDegrees=20, twistLimitDegrees=10}
 {bone1Name=dyn_frontHair_2_l, bone2Name=dyn_frontHair_1_l, swing1LimitDegrees=20, swing2LimitDegrees=30, twistLimitDegrees=10}
 {bone1Name=dyn_frontHair_3_l, bone2Name=dyn_frontHair_2_l, swing1LimitDegrees=20, swing2LimitDegrees=30, twistLimitDegrees=10}
 {bone1Name=dyn_frontHair_4_l, bone2Name=dyn_frontHair_3_l, swing1LimitDegrees=20, swing2LimitDegrees=30, twistLimitDegrees=10}
 {bone1Name=dyn_frontHair_1_r, bone2Name=base_frontHair_1_r, swing1LimitDegrees=20, swing2LimitDegrees=20, twistLimitDegrees=10}
 {bone1Name=dyn_frontHair_2_r, bone2Name=dyn_frontHair_1_r, swing1LimitDegrees=20, swing2LimitDegrees=20, twistLimitDegrees=10}
 {bone1Name=dyn_frontHair_3_r, bone2Name=dyn_frontHair_2_r, swing1LimitDegrees=20, swing2LimitDegrees=20, twistLimitDegrees=10}
 {bone1Name=dyn_frontHair_4_r, bone2Name=dyn_frontHair_3_r, swing1LimitDegrees=20, swing2LimitDegrees=20, twistLimitDegrees=10}
]
