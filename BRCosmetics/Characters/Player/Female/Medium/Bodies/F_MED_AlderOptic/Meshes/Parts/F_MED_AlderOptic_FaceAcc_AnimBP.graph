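# Animation Blueprint for the AlderOptic female face accessory, adding physics-driven secondary motion.
# It copies the parent mesh pose, replaces three chest collision bone transforms, runs a Control Rig, then simulates RigidBody physics relative to spine_05 with overridden gravity.
# Each update computes gravity and a front-end idle-montage weight from a curve.
# It also samples those bones' transforms from the body part cached on customization load, when that body has them.
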AnimBlueprint /BRCosmetics/Characters/Player/Female/Medium/Bodies/F_MED_AlderOptic/Meshes/Parts/F_MED_AlderOptic_FaceAcc_AnimBP
target_skeleton=&/F_MED_AlderOptic_FaceAcc_Skeleton
graphs [
 AnimGraph {nodes [
   n1 = RigidBody!(ComponentPose=Local To Component(LocalPose=Use cached pose 'inputCP'()), Alpha, OverrideWorldGravity:Vector="0,0,0") -> (Pose) {
    bClampLinearTranslationLimitToRefPose=true
    bOverrideWorldGravity=true
    overridePhysicsAsset=&/F_MED_AlderOptic_FaceAcc_Physics
    simulationSpace=BaseBoneSpace
    alphaBoolBlend {blendOption=Linear}
    alphaScaleBiasClamp {interpSpeedDecreasing=5, interpSpeedIncreasing=5}
    baseBoneRef {boneName=spine_05}
    componentAppliedLinearAccClamp=<3000 3000 3000>
    componentLinearAccScale=<0.4 0.4 0.4>
    componentLinearVelScale=<0.6 0.6 0.6>
    overrideWorldGravity {z=-980}
    simSpaceSettings {
     maxAngularAcceleration=500
     maxAngularVelocity=3
     maxLinearAcceleration=400
     maxLinearVelocity=250
     velocityScaleZ=0
     worldAlpha=1
     externalLinearDragV {x=1}
    }
   }
   Output Pose(Result=Component To Local(ComponentPose=@n1)) {layerGroup=None}
   n2 = "Transform (Modify) Bone - Bone: coll_chest_01_l"!(
    ComponentPose=Local To Component(LocalPose=Copy Pose From Mesh() {bUseAttachedParent=true})
    bAlphaBoolEnabled:bool=True
    Translation="0,0,0"
    Rotation="0,0,0"
   ) -> (Pose) {
    alphaInputType=Bool
    rotationMode=BMM_Replace
    rotationSpace=BCS_ParentBoneSpace
    translationMode=BMM_Replace
    translationSpace=BCS_ParentBoneSpace
    alphaBoolBlend {blendInTime=0.25, blendOption=Linear, blendOutTime=0.25}
    boneToModify {boneName=coll_chest_01_l}
   }
   n3 = "Transform (Modify) Bone - Bone: coll_chest_01_r"!(ComponentPose=@n2, bAlphaBoolEnabled:bool=True, Translation="0,0,0", Rotation="0,0,0") -> (Pose) {
    alphaInputType=Bool
    rotationMode=BMM_Replace
    rotationSpace=BCS_ParentBoneSpace
    translationMode=BMM_Replace
    translationSpace=BCS_ParentBoneSpace
    alphaBoolBlend {blendInTime=0.25, blendOption=Linear, blendOutTime=0.25}
    boneToModify {boneName=coll_chest_01_r}
   }
   n4 = "Transform (Modify) Bone - Bone: coll_breast_1_r"!(ComponentPose=@n3, bAlphaBoolEnabled:bool=True, Translation="0,0,0", Rotation="0,0,0") -> (Pose) {
    alphaInputType=Bool
    rotationMode=BMM_Replace
    rotationSpace=BCS_ParentBoneSpace
    translationMode=BMM_Replace
    translationSpace=BCS_ParentBoneSpace
    alphaBoolBlend {blendInTime=0.25, blendOption=Linear, blendOutTime=0.25}
    boneToModify {boneName=coll_breast_1_r}
   }
   n5 = Control Rig!(Source=Component To Local(ComponentPose=@n4), Alpha, isIdleMontage:double=0) -> (Pose) {controlRigAssetReference {blueprintRigClass=&/F_MED_AlderOptic_FaceAcc_CtrlRig.F_MED_AlderOptic_FaceAcc_CtrlRig_C}}
   Save cached pose 'inputCP'(Pose=@n5) {cacheName=inputCP}
  ]}
 EventGraph {nodes [
   Event BlueprintUpdateAnimation()
   > n1 = Sequence!(execute) -> (then_0, then_1, then_2:Exec)
   > n2 = "Gravity Override State Function (Parameter Asset) (Nativized)"(
    CharacterPartAnimInstance=Self-Reference()
    GameTimeInSeconds=GetGameTimeInSeconds()
    Parameters=/BRCosmetics/Characters/Player/Female/Medium/Bodies/F_MED_AlderOptic/Meshes/F_MED_AlderOptic_FaceAcc_Parameters
   )
   > Set GravityOverride(GravityOverride=@n2)
   n3 = GetAnimInstance(self=GetOwnerSkeletalMeshComponent())
   Set isIdleMontage(
    execute=@n1.then_1
    isIdleMontage:double=float * float!(
     A=GetCurveValue(self=@n3, CurveName=FrontEnd_Idle_Bypass)
     B:bool=Montage_IsPlaying(self=@n3, Montage=/BRCosmetics/Animation/Game/MainPlayer/Menu/BR/CustomSkins/_s40/AlderOptic_F/AlderOptic_Lobby_Idle_Montage)
     C:bool=Get bIsInFrontEnd()
    ) -> (ReturnValue) {is=title K2Node_PromotableOperator}
   ) -> (Output_Get:double)
   n4 = Get BodyMesh() -> (BodyMesh:Skeletal Mesh Component Budgeted Object Reference)
   Is Valid(exec=@n1.then_2, InputObject=@n4)
   n5 = GetSocketBoneName!(self:Skinned Mesh Component Object Reference=@n4, InSocketName:Name=coll_chest_01_l) -> (ReturnValue:Name=None)
   > n6 = Set isUsingItsBody(isUsingItsBody:bool="Not Equal Exactly (String)"(A="To String (Name)"(InName=@n5), B=None)) -> (Output_Get:bool)
   > Branch(Condition=@n6)
   n7 = Get BodyMesh()
   n8 = BreakTransform(InTransform=GetSocketTransform(self=@n7, InSocketName=coll_chest_01_l, TransformSpace=RTS_ParentBoneSpace))
   > Set coll_chest_01_l_Translation(coll_chest_01_l_Translation:Vector=@n8) -> (Output_Get:Vector)
   > Set coll_chest_01_l_Rotation(coll_chest_01_l_Rotation:Rotator=@n8.Rotation) -> (Output_Get:Rotator)
   n9 = BreakTransform(InTransform=GetSocketTransform(self=@n7, InSocketName=coll_chest_01_r, TransformSpace=RTS_ParentBoneSpace))
   > Set coll_chest_01_r_Translation(coll_chest_01_r_Translation:Vector=@n9) -> (Output_Get:Vector)
   > Set coll_chest_01_r_Rotation(coll_chest_01_r_Rotation:Rotator=@n9.Rotation) -> (Output_Get:Rotator)
   n10 = BreakTransform(InTransform=GetSocketTransform(self=@n7, InSocketName=coll_breast_1_r, TransformSpace=RTS_ParentBoneSpace))
   > Set coll_breast_1_r_Translation(coll_breast_1_r_Translation:Vector=@n10) -> (Output_Get:Vector)
   > Set coll_breast_1_r_Rotation(coll_breast_1_r_Rotation:Rotator=@n10.Rotation) -> (Output_Get:Rotator)
   n11 = Event OnCharacterCustomizationLoaded()
   > Set BodyMesh(BodyMesh:Skeletal Mesh Component Budgeted Object Reference=Get Array Item with EFort Custom Part Type(Array=@n11.CustomMeshParts)) -> (Output_Get:Skeletal Mesh Component Budgeted Object Reference)
  ]}
]
variables [
 BodyMesh {type=SkeletalMeshComponentBudgeted}
 coll_breast_1_r_Rotation {type=Rotator}
 coll_breast_1_r_Translation {type=Vector}
 coll_chest_01_l_Rotation {type=Rotator}
 coll_chest_01_l_Translation {type=Vector}
 coll_chest_01_r_Rotation {type=Rotator}
 coll_chest_01_r_Translation {type=Vector}
 GravityOverride {type=Vector}
 isIdleMontage {type=number}
 isUsingItsBody {type=boolean}
]
